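# WBP_PopupModal is a layout-only FortPopupModal widget template for Reboot Rally sidebar popups, with no event logic or variables of its own.
# Named slots expose the full-screen input block, background, image, header text, details text and QR code inside a centered stack.
# Hardware visibility borders swap the two inherited close buttons: a corner close button for touch input, otherwise a text button bound to the Close input action.
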WidgetBlueprint /FortniteCoreUI/UI/Sidebar/RebootRally/WBP_PopupModal
graphs [EventGraph {nodes [
   Event PreConstruct()
   Event Construct()
   Event Tick()
  ]}]
variables []
widgets {
 info {inheritedWidgetCount=2, parentClass=/Script/FortniteUI.FortPopupModal, rootWidgetClass=&Overlay, widgetCount=15}
 widgets [
  Overlay_1 {widgetClassPath=&Overlay, widget=Overlay_1}
  NamedSlot_InputBlock {
   widgetClassPath=&NamedSlot
   bIsVariable=true
   parent=Overlay_1
   widget=NamedSlot_InputBlock
   properties {visibility=Visible}
   slot {class=OverlaySlot, properties {horizontalAlignment=HAlign_Fill, verticalAlignment=VAlign_Fill}}
  }
  VBox {
   widgetClassPath=&VerticalBox
   parent=Overlay_1
   widget=VBox
   slot {class=OverlaySlot, properties {horizontalAlignment=HAlign_Center, verticalAlignment=VAlign_Center}}
  }
  Overlay_0 {widgetClassPath=&Overlay, parent=VBox, widget=Overlay_0, slot {class=VerticalBoxSlot}}
  NamedSlot_Background {
   widgetClassPath=&NamedSlot
   bIsVariable=true
   parent=Overlay_0
   widget=NamedSlot_Background
   slot {class=OverlaySlot, properties {horizontalAlignment=HAlign_Fill, verticalAlignment=VAlign_Fill}}
  }
  VerticalBox_27 {
   widgetClassPath=&VerticalBox
   parent=Overlay_0
   widget=VerticalBox_27
   slot {class=OverlaySlot, properties {horizontalAlignment=HAlign_Fill}}
  }
  NamedSlot_Image {
   widgetClassPath=&NamedSlot
   bIsVariable=true
   parent=VerticalBox_27
   widget=NamedSlot_Image
   slot {class=VerticalBoxSlot, properties {verticalAlignment=VAlign_Top}}
  }
  NamedSlot_HeaderText {widgetClassPath=&NamedSlot, bIsVariable=true, parent=VerticalBox_27, widget=NamedSlot_HeaderText, slot {class=VerticalBoxSlot}}
  NamedSlot_DetailsText {widgetClassPath=&NamedSlot, bIsVariable=true, parent=VerticalBox_27, widget=NamedSlot_DetailsText, slot {class=VerticalBoxSlot}}
  NamedSlot_QRCode {
   widgetClassPath=&NamedSlot
   bIsVariable=true
   parent=VerticalBox_27
   widget=NamedSlot_QRCode
   properties {navigation=&self:WidgetTree.NamedSlot_QRCode.WidgetNavigation_0}
   slot {class=VerticalBoxSlot}
  }
  CloseTouch {
   widgetClassPath=&Overlay
   parent=Overlay_0
   widget=CloseTouch
   properties {renderTransform {angle=0, scale=<1 1>, shear=<0 0>, translation=<48 -64>}}
   slot {class=OverlaySlot, properties {horizontalAlignment=HAlign_Right}}
  }
  VisibilityBorder_Touch {
   widgetClassPath=&CommonHardwareVisibilityBorder
   parent=CloseTouch
   widget=VisibilityBorder_Touch
   properties {visibilityQuery {
     autoDescription=" ANY( Input.Touch )"
     tokenStreamVersion=0
     userDescription=""
     queryTokenStream [0,1,1,1,0]
     tagDictionary [{tagName=Input.Touch}]
    }}
   slot {class=OverlaySlot}
  }
  Button_CloseTouch {
   widgetClassPath=/FortUILibrary/Foundation/Buttons/CloseButton
   bInherited=true
   bIsVariable=true
   parent=VisibilityBorder_Touch
   widget=Button_CloseTouch
   properties {
    onButtonBaseClicked=(&self:WidgetTree.Button_CloseTouch.HandleOnButtonClicked)
    onButtonBaseDoubleClicked=(&self:WidgetTree.Button_CloseTouch.HandleOnButtonDoubleClicked)
    onButtonBaseHovered=(&self:WidgetTree.Button_CloseTouch.HandleOnButtonHovered)
    onButtonBaseUnhovered=(&self:WidgetTree.Button_CloseTouch.HandleOnButtonUnhovered)
    onSelectedChangedBase=(&self:WidgetTree.Button_CloseTouch.HandleOnSelectedChanged)
    paddingOveride=false
   }
   slot {class=BorderSlot, properties {padding=<ltrb 0 0 0 0>}}
  }
  VisibilityBorder_NonTouch {
   widgetClassPath=&CommonHardwareVisibilityBorder
   parent=VBox
   widget=VisibilityBorder_NonTouch
   properties {visibilityQuery {
     autoDescription=" NONE( Input.Touch )"
     tokenStreamVersion=0
     userDescription=""
     queryTokenStream [0,1,3,1,0]
     tagDictionary [{tagName=Input.Touch}]
    }}
   slot {class=VerticalBoxSlot, properties {horizontalAlignment=HAlign_Right, verticalAlignment=VAlign_Top}}
  }
  IconTextButton_Close {
   widgetClassPath=&IconTextButton
   bInherited=true
   bIsVariable=true
   parent=VisibilityBorder_NonTouch
   widget=IconTextButton_Close
   properties {
    buttonText=Close
    navigation=&self:WidgetTree.IconTextButton_Close.WidgetNavigation_0
    onButtonBaseClicked=(&self:WidgetTree.IconTextButton_Close.HandleOnButtonClicked)
    onButtonBaseDoubleClicked=(&self:WidgetTree.IconTextButton_Close.HandleOnButtonDoubleClicked)
    onButtonBaseHovered=(&self:WidgetTree.IconTextButton_Close.HandleOnButtonHovered)
    onButtonBaseUnhovered=(&self:WidgetTree.IconTextButton_Close.HandleOnButtonUnhovered)
    onSelectedChangedBase=(&self:WidgetTree.IconTextButton_Close.HandleOnSelectedChanged)
    style=/Game/UI/Sidebar/RebootRally/RebootRallyPopupModalButtonStyle.RebootRallyPopupModalButtonStyle_C
    triggeringInputAction {dataTable=&InputActionDataTable, rowName=Close}
   }
   slot {class=BorderSlot, properties {padding=<ltrb 0 0 0 0>}}
  }
 ]
}
