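# WBP_RebootRallyPopupModal_HeaderText is a small UserWidget that supplies the header title for the Reboot Rally popup modal in the sidebar UI.
# An Overlay root holds one centered CommonTextBlock reading "Rally 'Em Back" in large, skewed, uppercase Burbank Black with bottom padding.
# The PreConstruct, Construct and Tick events are empty, so the widget is purely static layout with no logic or variables.
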WidgetBlueprint /FortniteCoreUI/UI/Sidebar/RebootRally/WBP_RebootRallyPopupModal_HeaderText
graphs [EventGraph {nodes [
   Event PreConstruct()
   Event Construct()
   Event Tick()
  ]}]
variables []
widgets {
 info {inheritedWidgetCount=0, parentClass=&UserWidget, rootWidgetClass=&Overlay, widgetCount=2}
 widgets [
  Overlay_Padding {widgetClassPath=&Overlay, widget=Overlay_Padding}
  Text_Header {
   widgetClassPath=&CommonTextBlock
   bIsVariable=true
   parent=Overlay_Padding
   widget=Text_Header
   properties {
    text=Rally 'Em Back
    textTransformPolicy=ToUpper
    font {
     bForceMonospaced=false
     bMaterialIsStencil=false
     fontMaterial=null
     fontObject=&Burbank
     letterSpacing=0
     monospacedWidth=1
     size=42
     skewAmount=0.15
     typefaceFontName=BurbankRegular-Black
     outlineSettings {
      bApplyOutlineToDropShadows=false
      bMiteredCorners=false
      bSeparateFillAlpha=false
      outlineMaterial=null
      outlineSize=0
      outlineColor=<rgba 0 0 0 1>
     }
    }
    shadowOffset=<0 0>
   }
   slot {class=OverlaySlot, properties {horizontalAlignment=HAlign_Center, padding=<ltrb 0 0 0 6>}}
  }
 ]
}
