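# Widget that provides the header image for the Reboot Rally popup modal in the sidebar UI.
# Nested overlays stack a cloud background material image, a spacer that sets a fixed size, and a slightly upscaled reward composition material image on top.
# Offsetting top and bottom padding shifts the content upward.
# It is purely visual, with only empty PreConstruct, Construct and Tick events and no variables or logic.
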WidgetBlueprint /FortniteCoreUI/UI/Sidebar/RebootRally/WBP_RebootRallyPopupModal_Image
graphs [EventGraph {nodes [
   Event PreConstruct()
   Event Construct()
   Event Tick()
  ]}]
variables []
widgets {
 info {inheritedWidgetCount=0, parentClass=&UserWidget, rootWidgetClass=&Overlay, widgetCount=6}
 widgets [
  Overlay_1 {widgetClassPath=&Overlay, widget=Overlay_1}
  Overlay_Padding {
   widgetClassPath=&Overlay
   parent=Overlay_1
   widget=Overlay_Padding
   slot {class=OverlaySlot, properties {horizontalAlignment=HAlign_Fill, verticalAlignment=VAlign_Fill, padding=<ltrb 0 -30 0 30>}}
  }
  Image_Background {
   widgetClassPath=&Image
   bIsVariable=true
   parent=Overlay_Padding
   widget=Image_Background
   properties {brush {
     bIsDynamicallyLoaded=false
     drawAs=Image
     imageType=FullColor
     mirroring=NoMirror
     resourceName=None
     resourceObject=/Game/UI/Sidebar/RebootRally/Materials/MI_UI_RebootRally_BGClouds_S32_PopupModal
     tiling=NoTile
     imageSize=<0 0>
     margin=<ltrb 0 0 0 0>
     outlineSettings {
      bUseBrushTransparency=false
      roundingType=HalfHeightRadius
      width=0
      color {colorUseRule=UseColor_Specified, specifiedColor=<rgba 0 0 0 0>}
      cornerRadii=<0 0 0 0>
     }
     tintColor {colorUseRule=UseColor_Specified, specifiedColor=<rgba 1 1 1 1>}
     uVRegion {bIsValid=false, max=<0 0>, min=<0 0>}
    }}
   slot {class=OverlaySlot, properties {horizontalAlignment=HAlign_Fill, verticalAlignment=VAlign_Fill}}
  }
  Spacer_37 {
   widgetClassPath=&Spacer
   parent=Overlay_Padding
   widget=Spacer_37
   properties {size=<618 300>}
   slot {class=OverlaySlot, properties {horizontalAlignment=HAlign_Fill, verticalAlignment=VAlign_Fill}}
  }
  Overlay_Rewards {
   widgetClassPath=&Overlay
   parent=Overlay_Padding
   widget=Overlay_Rewards
   slot {class=OverlaySlot, properties {horizontalAlignment=HAlign_Fill, verticalAlignment=VAlign_Fill}}
  }
  Image_Rewards {
   widgetClassPath=&Image
   bIsVariable=true
   parent=Overlay_Rewards
   widget=Image_Rewards
   properties {
    brush {
     bIsDynamicallyLoaded=false
     drawAs=Image
     imageType=NoImage
     mirroring=NoMirror
     resourceName=None
     resourceObject=/Game/UI/Sidebar/RebootRally/Materials/MI_UI_Reward_Composition_Popup
     tiling=NoTile
     imageSize=<32 32>
     margin=<ltrb 0 0 0 0>
     outlineSettings {
      bUseBrushTransparency=false
      roundingType=HalfHeightRadius
      width=0
      color {colorUseRule=UseColor_Specified, specifiedColor=<rgba 0 0 0 0>}
      cornerRadii=<0 0 0 0>
     }
     tintColor {colorUseRule=UseColor_Specified, specifiedColor=<rgba 1 1 1 1>}
     uVRegion {bIsValid=false, max=<0 0>, min=<0 0>}
    }
    renderTransform {angle=0, scale=<1.1 1.1>, shear=<0 0>, translation=<0 0>}
   }
   slot {class=OverlaySlot, properties {horizontalAlignment=HAlign_Fill, verticalAlignment=VAlign_Fill}}
  }
 ]
}
