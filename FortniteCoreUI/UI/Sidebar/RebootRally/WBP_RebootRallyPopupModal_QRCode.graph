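# QR code widget for the Reboot Rally popup modal, derived from RebootRallyQRCode, with a scannable code and clickable link.
# On construct, SetIsInvite or an input method change, InputMethodChanged updates the URL, prompt text and QR texture.
# URL and texture depend on invite versus rally mode and on input type; the prompt shows only for mouse/keyboard and touch.
# Switch clears the URL, hides the prompt and blocks launching the URL on click.
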WidgetBlueprint /FortniteCoreUI/UI/Sidebar/RebootRally/WBP_RebootRallyPopupModal_QRCode
graphs [
 EventGraph {nodes [
   Event Construct()
   > InputMethodChanged!(execute, self, NewInputType:ECommonInputType Enum=GetCurrentInputType(self=Get CommonInputSubsystem())) -> (then)
   Event On Clicked()
   > n1 = Branch(Condition=Get bIsSwitch() -> (bIsSwitch:bool))
   LaunchURL!(execute=@n1.else, URL:String=Get QRCodeURL() -> (QRCodeURL:String)) -> (then)
   Event OnInitialized()
   > Set bIsSwitch(bIsSwitch:bool="Equal, Case Insensitive (String)"(A=Switch, B=GetPlatformName!() -> (ReturnValue:String))) -> (Output_Get:bool)
   > Assign On Input Method Changed!(execute:Exec, self:Common Input Object Reference=Get CommonInputSubsystem(), Delegate:"Delegate (by ref)"=Create Event()) -> (then:Exec) {is=title K2Node_AssignDelegate}
   n2 = Event SetIsInvite!() -> (OutputDelegate, then, bIsInvite:bool) {is=title K2Node_Event}
   > Set bIsInvite(bIsInvite:bool=@n2.bIsInvite) -> (Output_Get:bool)
   > InputMethodChanged(NewInputType=GetCurrentInputType(self=Get CommonInputSubsystem()))
  ]}
 InputMethodChanged {nodes [
   n1 = InputMethodChanged!() -> (then, NewInputType:ECommonInputType Enum) {is=title K2Node_FunctionEntry}
   > UpdateURL!(execute, self, NewInputType:ECommonInputType Enum=@n1) -> (then)
   > UpdateText!(execute, self, NewInputType:ECommonInputType Enum=@n1) -> (then)
   > UpdateQRCode!(execute, self, NewInputType:ECommonInputType Enum=@n1) -> (then)
  ]}
 UpdateQRCode {nodes [
   n1 = UpdateQRCode!() -> (then, NewInputType:ECommonInputType Enum) {is=title K2Node_FunctionEntry}
   > n2 = Branch(Condition=Get bIsInvite() -> (bIsInvite:bool))
   n3 = Get Image_QRCode() -> (Image_QRCode:Image Object Reference)
   > SetBrushFromTexture!(
    execute
    self:Image Object Reference=@n3
    Texture:Texture 2D Object Reference=Select!(
     MouseAndKeyboard:Texture 2D Object Reference=/Game/UI/Sidebar/RebootRally/Textures/T_UI_RebootRallyQRCodeInviteKBM
     Gamepad:Texture 2D Object Reference=/Game/UI/Sidebar/RebootRally/Textures/T_UI_RebootRallyQRCodeInviteController
     Touch:Texture 2D Object Reference=/Game/UI/Sidebar/RebootRally/Textures/T_UI_RebootRallyQRCodeInviteMobile
     Count:Texture 2D Object Reference=/Game/UI/Sidebar/RebootRally/Textures/T_UI_RebootRallyQRCode
     Index:ECommonInputType Enum=@n1
    ) -> (ReturnValue:Texture 2D Object Reference)
    bMatchSize:bool=false
   ) -> (then)
   SetBrushFromTexture(
    execute=@n2.else
    self=@n3
    Texture=Select(
     MouseAndKeyboard=/Game/UI/Sidebar/RebootRally/Textures/T_UI_RebootRallyQRCodeRallyEmKBM
     Gamepad=/Game/UI/Sidebar/RebootRally/Textures/T_UI_RebootRallyQRCodeRallyEmController
     Touch=/Game/UI/Sidebar/RebootRally/Textures/T_UI_RebootRallyQRCodeRallyEmMobile
     Count=&T_UI_RebootRallyQRCode
     Index=@n1
    )
    bMatchSize=false
   )
  ]}
 UpdateText {nodes [
   n1 = UpdateText!() -> (then, NewInputType:ECommonInputType Enum) {is=title K2Node_FunctionEntry}
   > n2 = Branch(Condition=Get bIsSwitch())
   n3 = Get Text_QRAction() -> (Text_QRAction:Common Text Object Reference)
   > SetVisibility(self=@n3, InVisibility=Collapsed)
   Set Text(
    execute=@n2.else
    Text=Select(MouseAndKeyboard:Text="LOCTEXT(\"Scan or Click\")", Gamepad:Text, Touch:Text="LOCTEXT(\"Scan or Tap\")", Count:Text, Index=@n1) -> (ReturnValue:Text)
    self=@n3
   )
   > SetVisibility(
    self=@n3
    InVisibility=Select(
     MouseAndKeyboard:ESlateVisibility Enum=SelfHitTestInvisible
     Gamepad:ESlateVisibility Enum=Collapsed
     Touch:ESlateVisibility Enum=SelfHitTestInvisible
     Count:ESlateVisibility Enum=Collapsed
     Index=@n1
    ) -> (ReturnValue:ESlateVisibility Enum)
   )
  ]}
 UpdateURL {nodes [
   n1 = UpdateURL!() -> (then, NewInputType:ECommonInputType Enum) {is=title K2Node_FunctionEntry}
   > n2 = Branch(Condition=Get bIsSwitch())
   > Set QRCodeURL(QRCodeURL:String) -> (Output_Get:String)
   n3 = Branch(execute=@n2.else, Condition=Get bIsInvite())
   > Set QRCodeURL(QRCodeURL=Select(
     MouseAndKeyboard:String=https://fn.gg/mouseClickRebootinvite
     Gamepad:String=https://fn.gg/controllerQRrebootinvite
     Touch:String=https://fn.gg/mobileTapRebootinvite
     Count:String=https://fn.gg/rebootrally
     Index=@n1
    ) -> (ReturnValue:String))
   Set QRCodeURL(
    execute=@n3.else
    QRCodeURL=Select(
     MouseAndKeyboard=https://fn.gg/mouseClickRebootrallyback
     Gamepad=https://fn.gg/controllerQRrebootrallyback
     Touch=https://fn.gg/mobileTapRebootrallyback
     Count=https://fn.gg/rebootrally
     Index=@n1
    )
   )
  ]}
]
variables [
 bIsInvite {type=boolean}
 bIsSwitch {type=boolean}
 QRCodeURL {type=string, default=https://fn.gg/rebootrally}
]
widgets {
 info {inheritedWidgetCount=1, parentClass=/Script/FortniteUI.RebootRallyQRCode, rootWidgetClass=&VerticalBox, widgetCount=6}
 widgets [
  VBox_QRCode {
   widgetClassPath=&VerticalBox
   widget=VBox_QRCode
   properties {navigation=&self:WidgetTree.VBox_QRCode.WidgetNavigation_0}
  }
  Border_Background {
   widgetClassPath=&Border
   parent=VBox_QRCode
   widget=Border_Background
   properties {
    navigation=&self:WidgetTree.Border_Background.WidgetNavigation_0
    visibility=SelfHitTestInvisible
    background {
     bIsDynamicallyLoaded=false
     drawAs=Image
     imageType=NoImage
     mirroring=NoMirror
     resourceName=None
     resourceObject=/Game/UI/Sidebar/RebootRally/Materials/MI_UI_RebootRallyQRCodeBackground
     tiling=NoTile
     imageSize=<32 32>
     margin=<ltrb 0 0 0 0>
     outlineSettings {
      bUseBrushTransparency=false
      roundingType=HalfHeightRadius
      width=0
      color {colorUseRule=UseColor_Specified, specifiedColor=<rgba 0 0 0 0>}
      cornerRadii=<0 0 0 0>
     }
     tintColor {colorUseRule=UseColor_Specified, specifiedColor=<rgba 1 1 1 1>}
     uVRegion {bIsValid=false, max=<0 0>, min=<0 0>}
    }
    padding=<ltrb 10 10 10 10>
   }
   slot {class=VerticalBoxSlot, properties {horizontalAlignment=HAlign_Center, verticalAlignment=VAlign_Center}}
  }
  Image_QRCode {
   widgetClassPath=&Image
   bIsVariable=true
   parent=Border_Background
   widget=Image_QRCode
   properties {
    navigation=&self:WidgetTree.Image_QRCode.WidgetNavigation_0
    brush {
     bIsDynamicallyLoaded=false
     drawAs=Image
     imageType=NoImage
     mirroring=NoMirror
     resourceName=None
     resourceObject=&T_UI_RebootRallyQRCode
     tiling=NoTile
     imageSize=<128 128>
     margin=<ltrb 0 0 0 0>
     outlineSettings {
      bUseBrushTransparency=false
      roundingType=HalfHeightRadius
      width=0
      color {colorUseRule=UseColor_Specified, specifiedColor=<rgba 0 0 0 0>}
      cornerRadii=<0 0 0 0>
     }
     tintColor {colorUseRule=UseColor_Specified, specifiedColor=<rgba 1 1 1 1>}
     uVRegion {bIsValid=false, max=<0 0>, min=<0 0>}
    }
   }
   slot {class=BorderSlot, properties {padding=<ltrb 10 10 10 10>}}
  }
  Text_QRAction {
   widgetClassPath=&CommonTextBlock
   bIsVariable=true
   parent=VBox_QRCode
   widget=Text_QRAction
   properties {
    navigation=&self:WidgetTree.Text_QRAction.WidgetNavigation_0
    text=Scan or click
    textTransformPolicy=ToUpper
    font {
     bForceMonospaced=false
     bMaterialIsStencil=false
     fontMaterial=null
     fontObject=&BurbankSmall
     letterSpacing=0
     monospacedWidth=1
     size=18
     skewAmount=0
     typefaceFontName=Bold
     outlineSettings {
      bApplyOutlineToDropShadows=false
      bMiteredCorners=false
      bSeparateFillAlpha=false
      outlineMaterial=null
      outlineSize=0
      outlineColor=<rgba 0 0 0 1>
     }
    }
    shadowOffset=<0 0>
   }
   slot {class=VerticalBoxSlot, properties {horizontalAlignment=HAlign_Center, verticalAlignment=VAlign_Top, padding=<ltrb 0 10 0 0>}}
  }
  Spacer_39 {
   widgetClassPath=&Spacer
   parent=VBox_QRCode
   widget=Spacer_39
   properties {navigation=&self:WidgetTree.Spacer_39.WidgetNavigation_0, size=<0 40>}
   slot {class=VerticalBoxSlot, properties {verticalAlignment=VAlign_Top}}
  }
  InputActionWidget {widgetClassPath=&CommonActionWidget, bInherited=true, widget=null}
 ]
}
