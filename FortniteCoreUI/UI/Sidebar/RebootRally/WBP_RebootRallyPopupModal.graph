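# Reboot Rally popup modal widget, built on RebootRallyQRModalWidget, that presents a QR code with header, details text and image.
# Child widgets fill WBP_PopupModal's named slots, alongside a background blur and a close button inside a border shown only without touch input.
# Construct sends a quest stat event tagged Athena.Quests.RebootRally.Frontend.
# OnSetIsInvite switches the header between new-player and returning-player strings, despite their DetailsText variable names.
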WidgetBlueprint /FortniteCoreUI/UI/Sidebar/RebootRally/WBP_RebootRallyPopupModal
graphs [EventGraph {nodes [
   Event Construct()
   > n1 = Cast To FortPlayerController(Object=GetOwningPlayer())
   > SendClientStatEventWithTags!(
    execute:Exec
    QuestManager:Fort Quest Manager Object Reference=GetQuestManager!(self:Fort Player Controller Object Reference=@n1, SubGame:ESubGame Enum=Invalid) -> (ReturnValue:Fort Quest Manager Object Reference)
    TargetTags:Gameplay Tag Container Structure
    SourceTags:Gameplay Tag Container Structure=MakeLiteralGameplayTagContainer!(Value:Gameplay Tag Container Structure=(GameplayTags=((TagName="Athena.Quests.RebootRally.Frontend")))) -> (ReturnValue:Gameplay Tag Container Structure)
    ContextTags:Gameplay Tag Container Structure
    Count:int=1
   ) -> (then:Exec, OnSuccess:Exec, OnIgnored:Exec, OnFailure:Exec) {is=title K2Node_AsyncAction}
   n2 = Event OnSetIsInvite!() -> (OutputDelegate, then, bIsInvited:bool) {is=title K2Node_Event}
   > Set Text(
    Text=Select(
     Option 0:"Text (by ref)"=Get DetailsTextOldPlayer() -> (DetailsTextOldPlayer:Text)
     Option 1:"Text (by ref)"=Get DetailsTextNewPlayer() -> (DetailsTextNewPlayer:Text)
     Index=@n2.bIsInvited
    ) -> (ReturnValue:"Text (by ref)")
    self=Get Text_Header!(self:WBP Reboot Rally Popup Modal Header Text Object Reference=Get WBP_RebootRallyPopupModal_HeaderText() -> (WBP_RebootRallyPopupModal_HeaderText:WBP Reboot Rally Popup Modal Header Text Object Reference)) -> (Text_Header:Common Text Object Reference)
   )
   Event PreConstruct()
   Event Tick()
  ]}]
variables [
 DetailsTextNewPlayer {type=string, default=Rally 'Em}
 DetailsTextOldPlayer {type=string, default=Rally 'Em Back}
]
widgets {
 info {
  inheritedWidgetCount=3
  namedSlotCount=6
  parentClass=/Script/FortniteUI.RebootRallyQRModalWidget
  rootWidgetClass=&/WBP_PopupModal.WBP_PopupModal_C
  widgetCount=10
 }
 widgets [
  WBP_PopupModal {
   widgetClassPath=&/WBP_PopupModal
   bInherited=true
   bIsVariable=true
   widget=WBP_PopupModal
   properties {navigation=&self:WidgetTree.WBP_PopupModal.WidgetNavigation_0}
  }
  WBP_RebootRallyPopupModal_Background {
   widgetClassPath=&/WBP_RebootRallyPopupModal_Background
   bIsVariable=true
   namedSlotHost=WBP_PopupModal
   widget=WBP_RebootRallyPopupModal_Background
  }
  WBP_RebootRallyPopupModal_DetailsText {
   widgetClassPath=&/WBP_RebootRallyPopupModal_DetailsText
   bIsVariable=true
   namedSlotHost=WBP_PopupModal
   widget=WBP_RebootRallyPopupModal_DetailsText
  }
  WBP_RebootRallyPopupModal_HeaderText {
   widgetClassPath=&/WBP_RebootRallyPopupModal_HeaderText
   bIsVariable=true
   namedSlotHost=WBP_PopupModal
   widget=WBP_RebootRallyPopupModal_HeaderText
  }
  WBP_RebootRallyPopupModal_Image {
   widgetClassPath=&/WBP_RebootRallyPopupModal_Image
   bIsVariable=true
   namedSlotHost=WBP_PopupModal
   widget=WBP_RebootRallyPopupModal_Image
  }
  Overlay_0 {widgetClassPath=&Overlay, namedSlotHost=WBP_PopupModal, widget=Overlay_0}
  BackgroundBlur {
   widgetClassPath=&BackgroundBlur
   parent=Overlay_0
   widget=BackgroundBlur
   properties {blurStrength=10}
   slot {class=OverlaySlot, properties {horizontalAlignment=HAlign_Fill, verticalAlignment=VAlign_Fill}}
  }
  VisibilityBorder_NonTouch {
   widgetClassPath=&CommonHardwareVisibilityBorder
   parent=Overlay_0
   widget=VisibilityBorder_NonTouch
   properties {visibilityQuery {
     autoDescription=" NONE( Input.Touch )"
     tokenStreamVersion=0
     userDescription=""
     queryTokenStream [0,1,3,1,0]
     tagDictionary [{tagName=Input.Touch}]
    }}
   slot {class=OverlaySlot, properties {horizontalAlignment=HAlign_Fill, verticalAlignment=VAlign_Fill}}
  }
  BackgroundCloseButton {
   widgetClassPath=/Game/Athena/UI/Matchmaking/SimpleCommonButton
   bInherited=true
   bIsVariable=true
   parent=VisibilityBorder_NonTouch
   widget=BackgroundCloseButton
   properties {
    onButtonBaseClicked=(&self:WidgetTree.BackgroundCloseButton.HandleOnButtonClicked)
    onButtonBaseDoubleClicked=(&self:WidgetTree.BackgroundCloseButton.HandleOnButtonDoubleClicked)
    onButtonBaseHovered=(&self:WidgetTree.BackgroundCloseButton.HandleOnButtonHovered)
    onButtonBaseUnhovered=(&self:WidgetTree.BackgroundCloseButton.HandleOnButtonUnhovered)
    onSelectedChangedBase=(&self:WidgetTree.BackgroundCloseButton.HandleOnSelectedChanged)
    style=/FortUILibrary/Foundation/Buttons/ButtonStyles/ButtonStyle-Empty.ButtonStyle-Empty_C
   }
   slot {class=BorderSlot, properties {padding=<ltrb 0 0 0 0>}}
  }
  WBP_RebootRallyPopupModal_QRCode {
   widgetClassPath=&/WBP_RebootRallyPopupModal_QRCode
   bInherited=true
   bIsVariable=true
   namedSlotHost=WBP_PopupModal
   widget=WBP_RebootRallyPopupModal_QRCode
   properties {
    bIsFocusable=false
    navigation=&self:WidgetTree.WBP_RebootRallyPopupModal_QRCode.WidgetNavigation_0
   }
  }
 ]
}
