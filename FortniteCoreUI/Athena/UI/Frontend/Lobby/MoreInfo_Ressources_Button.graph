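# MoreInfo_Ressources_Button is a CommonButtonBase widget for the frontend lobby: a small round "?" info button with the inherited input action widget beside it.
# A fixed-size Border draws the circle with an info-button material around the text.
# On hover, one short animation scales up the glyph, darkens the content color and animates the material's Edge and InnerEdge parameters.
# Unhover plays the same animation in reverse.
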WidgetBlueprint /FortniteCoreUI/Athena/UI/Frontend/Lobby/MoreInfo_Ressources_Button
animations [Hover {
  binding_count=2
  length_seconds=0.15
  object=&self:Hover
  sequence {bindings [
    CommonTextBlock_60 {
     id=#1
     possessed_class=CommonTextBlock
     tracks [{
       type=MovieScene2DTransformTrack
       sections [{
         type=MovieScene2DTransformSection
         end_frame=3
         start_frame=0
         channels [
          Translation.X {type=MovieSceneScriptingFloatChannel}
          Translation.Y {type=MovieSceneScriptingFloatChannel}
          Angle {type=MovieSceneScriptingFloatChannel}
          Scale.X {type=MovieSceneScriptingFloatChannel, keys [[0,0.66],[3,1]]}
          Scale.Y {type=MovieSceneScriptingFloatChannel, keys [[0,0.66],[3,1]]}
          Shear.X {type=MovieSceneScriptingFloatChannel}
          Shear.Y {type=MovieSceneScriptingFloatChannel}
         ]
        }]
      }]
    }
    Border_27 {
     id=#2
     possessed_class=Border
     tracks [
      {
       type=MovieSceneWidgetMaterialTrack
       sections [{
         type=MovieSceneParameterSection
         end_frame=3
         start_frame=0
         channels [
          Edge {type=MovieSceneScriptingFloatChannel, keys [[0,0.2],[3,0]]}
          InnerEdge {type=MovieSceneScriptingFloatChannel, keys [[0,-0.1167],[3,0.6533]]}
         ]
        }]
      }
      {
       type=MovieSceneColorTrack
       display_name=Content Color and Opacity
       sections [{
         type=MovieSceneColorSection
         end_frame=3
         start_frame=0
         channels [
          Color.R {type=MovieSceneScriptingFloatChannel, keys [[0,1],[3,0]]}
          Color.G {type=MovieSceneScriptingFloatChannel, keys [[0,1],[3,0]]}
          Color.B {type=MovieSceneScriptingFloatChannel, keys [[0,1],[3,0]]}
          Color.A {type=MovieSceneScriptingFloatChannel}
         ]
        }]
      }
     ]
    }
   ]}
 }]
graphs [EventGraph {nodes [
   Event On Hovered()
   n1 = Get Hover() -> (Hover:Widget Animation Object Reference)
   > PlayAnimationForward(InAnimation=@n1)
   Event On Unhovered()
   > PlayAnimationReverse(InAnimation=@n1)
  ]}]
variables []
widgets {
 info {inheritedWidgetCount=1, parentClass=&CommonButtonBase, rootWidgetClass=&HorizontalBox, widgetCount=5}
 widgets [
  HorizontalBox_25 {widgetClassPath=&HorizontalBox, widget=HorizontalBox_25}
  SizeBox_0 {
   widgetClassPath=&SizeBox
   parent=HorizontalBox_25
   widget=SizeBox_0
   properties {bOverride_HeightOverride=true, bOverride_WidthOverride=true, heightOverride=32, widthOverride=32}
   slot {class=HorizontalBoxSlot}
  }
  Border_27 {
   widgetClassPath=&Border
   parent=SizeBox_0
   widget=Border_27
   properties {
    horizontalAlignment=HAlign_Center
    verticalAlignment=VAlign_Center
    background {
     bIsDynamicallyLoaded=false
     drawAs=Image
     imageType=NoImage
     mirroring=NoMirror
     resourceName=None
     resourceObject=/Game/Athena/UI/InGame/Interrogation/Materials/M_UI_Circle_Infobutton
     tiling=NoTile
     imageSize=<28 28>
     margin=<ltrb 0 0 0 0>
     outlineSettings {
      bUseBrushTransparency=false
      roundingType=HalfHeightRadius
      width=0
      color {colorUseRule=UseColor_Specified, specifiedColor=<rgba 0 0 0 0>}
      cornerRadii=<0 0 0 0>
     }
     tintColor {colorUseRule=UseColor_Specified, specifiedColor=<rgba 1 1 1 1>}
     uVRegion {bIsValid=false, max=<0 0>, min=<0 0>}
    }
    padding=<ltrb 4 -16 4 -16>
   }
   slot {class=SizeBoxSlot}
  }
  CommonTextBlock_60 {
   widgetClassPath=&CommonTextBlock
   parent=Border_27
   widget=CommonTextBlock_60
   properties {
    lineHeightPercentage=0.9429
    text=?
    font {
     bForceMonospaced=false
     bMaterialIsStencil=false
     fontMaterial=null
     fontObject=&BurbankSmall
     letterSpacing=0
     monospacedWidth=1
     size=19
     skewAmount=0
     typefaceFontName=Bold
     outlineSettings {
      bApplyOutlineToDropShadows=false
      bMiteredCorners=false
      bSeparateFillAlpha=false
      outlineMaterial=null
      outlineSize=0
      outlineColor=<rgba 0 0 0 1>
     }
    }
    renderTransform {angle=0, scale=<0.66 0.66>, shear=<-8 0>, translation=<0 0>}
    shadowOffset=<0 0>
   }
   slot {class=BorderSlot, properties {horizontalAlignment=HAlign_Center, verticalAlignment=VAlign_Center, padding=<ltrb 4 -16 4 -16>}}
  }
  InputActionWidget {
   widgetClassPath=&CommonActionWidget
   bInherited=true
   bIsVariable=true
   parent=HorizontalBox_25
   widget=InputActionWidget
   slot {class=HorizontalBoxSlot, properties {verticalAlignment=VAlign_Center, padding=<ltrb 0 -16 0 -16>}}
  }
 ]
}
