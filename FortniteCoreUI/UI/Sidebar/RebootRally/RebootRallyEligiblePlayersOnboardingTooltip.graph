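# Sidebar onboarding tooltip that prompts the player to invite eligible friends or new players to Reboot Rally, with a dismiss button.
# PreConstruct resizes the button for mobile and applies FlipHorizontal, mirroring the bubble while counter-mirroring text and button to keep them readable.
# OnEndTooltip plays Outro, notifies the native parent on finish, then plays Reset to restore appearance.
# A Pulse burst animation is defined but not played in this graph.
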WidgetBlueprint /FortniteCoreUI/UI/Sidebar/RebootRally/RebootRallyEligiblePlayersOnboardingTooltip
animations [
 Pulse {
  binding_count=1
  length_seconds=1.8
  object=&self:Pulse
  sequence {bindings [ImageBurst {
     id=#1
     possessed_class=Image
     tracks [
      {
       type=MovieSceneWidgetMaterialTrack
       sections [{
         type=MovieSceneParameterSection
         channels [
          Border_Offset_Top {type=MovieSceneScriptingFloatChannel, keys [[0,-0.1345],[20,-0.1145]]}
          Border_Offset_Right {type=MovieSceneScriptingFloatChannel, keys [[0,-0.1445],[20,-0.0745]]}
          Border_Offset_Left {type=MovieSceneScriptingFloatChannel, keys [[0,-0.1064],[20,-0.0564]]}
          Border_Offset_Bottom {type=MovieSceneScriptingFloatChannel, const [-0.1155,1]}
         ]
        }]
      }
      {
       type=MovieSceneFloatTrack
       display_name=Render Opacity
       sections [{type=MovieSceneFloatSection, channels [None {type=MovieSceneScriptingFloatChannel, keys [[-1,0],[0,1],[20,0],[35,0]]}]}]
      }
      {
       type=MovieScene2DTransformTrack
       sections [{
         type=MovieScene2DTransformSection
         channels [
          Scale.X {type=MovieSceneScriptingFloatChannel, keys [[0,1],[20,1.25]]}
          Scale.Y {type=MovieSceneScriptingFloatChannel, keys [[0,1],[20,1.4]]}
         ]
        }]
      }
     ]
    }]}
 }
 Outro {
  binding_count=3
  length_seconds=0.3
  object=&self:Outro
  sequence {bindings [
    CommonRichTextBlock_140 {
     id=#2
     possessed_class=CommonRichTextBlock
     tracks [{
       type=MovieSceneFloatTrack
       display_name=Render Opacity
       sections [{type=MovieSceneFloatSection, channels [None {type=MovieSceneScriptingFloatChannel, keys [[0,1],[6,0]]}]}]
      }]
    }
    Overlay_166 {
     id=#3
     possessed_class=Overlay
     tracks [{
       type=MovieSceneFloatTrack
       display_name=Render Opacity
       sections [{type=MovieSceneFloatSection, channels [None {type=MovieSceneScriptingFloatChannel, thinned_from=3, keys [[1,1],[6,0]]}]}]
      }]
    }
    SizeBox_9 {
     id=#4
     possessed_class=SizeBox
     tracks [{
       type=MovieScene2DTransformTrack
       sections [{type=MovieScene2DTransformSection, channels [Scale.Y {type=MovieSceneScriptingFloatChannel, keys [[0,1],[2,1],[4,0]]}]}]
      }]
    }
   ]}
 }
 Reset {
  binding_count=3
  length_seconds=0.05
  object=&self:Reset
  sequence {bindings [
    SizeBox_9 {
     id=#5
     possessed_class=SizeBox
     tracks [{
       type=MovieScene2DTransformTrack
       sections [{type=MovieScene2DTransformSection, end_frame=0, start_frame=0, channels [Scale.Y {type=MovieSceneScriptingFloatChannel, const [1,1]}]}]
      }]
    }
    Overlay_166 {
     id=#6
     possessed_class=Overlay
     tracks [{
       type=MovieSceneFloatTrack
       display_name=Render Opacity
       sections [{type=MovieSceneFloatSection, end_frame=0, start_frame=0, channels [None {type=MovieSceneScriptingFloatChannel, const [1,1]}]}]
      }]
    }
    CommonRichTextBlock_140 {
     id=#7
     possessed_class=CommonRichTextBlock
     tracks [{
       type=MovieSceneFloatTrack
       display_name=Render Opacity
       sections [{type=MovieSceneFloatSection, end_frame=0, start_frame=0, channels [None {type=MovieSceneScriptingFloatChannel, const [1,1]}]}]
      }]
    }
   ]}
 }
]
graphs [
 EventGraph {nodes [
   Event OnEndTooltip() {is=title K2Node_Event}
   > n1 = "Play Animation with Finished event (legacy)"!(
    execute:Exec
    Widget:User Widget Object Reference=Self-Reference()
    InAnimation:Widget Animation Object Reference=Get Outro() -> (Outro:Widget Animation Object Reference)
    StartAtTime:float=0
    NumLoopsToPlay:int=1
    PlayMode:EUMGSequencePlayMode Enum
    PlaybackSpeed:float=1
   ) -> (then:Exec, Result:UMGSequence Player Object Reference, Finished:Exec) {is=title K2Node_PlayAnimation}
   NotifyCloseAnimationFinished(execute=@n1.Finished, self:Fort Sidebar Onboard Tooltip Widget Object Reference)
   > PlayAnimation(InAnimation=Get Reset() -> (Reset:Widget Animation Object Reference))
   Event PreConstruct()
   > n2 = Sequence()
   SetTooltipAlignment(execute=@n2.then_1)
   n3 = IsMobileGame()
   SetUserSpecifiedScale!(
    execute=@n2.then_0
    self:Scale Box Object Reference=Get ScaleBox_mobile() -> (ScaleBox_mobile:Scale Box Object Reference)
    InUserSpecifiedScale:float=Select(Option 0:float=1, Option 1:float=1.2, Index=@n3) -> (ReturnValue:float)
   ) -> (then)
   > n4 = Branch(Condition=@n3)
   > SetHeightOverride(self=Get SizeBox_mobile() -> (SizeBox_mobile:Size Box Object Reference), InHeightOverride=60)
   ClearHeightOverride(execute=@n4.else, self:Size Box Object Reference=Get SizeBox_mobile())
   Event Tick()
   Event OnShowTooltip() {is=title K2Node_Event}
   Event Construct()
  ]}
 SetTooltipAlignment {nodes [
   SetTooltipAlignment() {is=title K2Node_FunctionEntry}
   n1 = Get FlipHorizontal() -> (FlipHorizontal:bool)
   n2 = Make Widget Transform!(
    Translation:Vector 2D Structure=(X=0,Y=0)
    Scale:Vector 2D Structure=Select(Option 0:Vector 2D Structure=(X=1,Y=1), Option 1:Vector 2D Structure=(X=-1,Y=1), Index=@n1) -> (ReturnValue:Vector 2D Structure)
    Shear:Vector 2D Structure=(X=0,Y=0)
    Angle:float=0
   ) -> (WidgetTransform:Widget Transform Structure) {is=title K2Node_MakeStruct}
   > SetRenderTransform!(
    execute
    self:Widget Object Reference=Get Overlay_166() -> (Overlay_166:Overlay Object Reference)
    InTransform:Widget Transform Structure=@n2
   ) -> (then)
   > SetRenderTransform(
    self=[
     Get Dismiss_Button() -> (Dismiss_Button:Common Button Legacy Object Reference)
     Get CommonRichTextBlock_106() -> (CommonRichTextBlock_106:Common Rich Text Block Object Reference)
     Get CommonRichTextBlock_140() -> (CommonRichTextBlock_140:Common Rich Text Block Object Reference)
    ]
    InTransform=@n2
   )
   > Set HorizontalAlignment!(
    execute
    HorizontalAlignment:EHorizontalAlignment Enum=Select(Option 0:EHorizontalAlignment Enum=HAlign_Right, Option 1:EHorizontalAlignment Enum=HAlign_Left, Index=@n1) -> (ReturnValue:EHorizontalAlignment Enum)
    self:Vertical Box Slot Object Reference=SlotAsVerticalBoxSlot(Widget=Get ScaleBox_mobile())
   ) -> (then, Output_Get:EHorizontalAlignment Enum=HAlign_Fill)
  ]}
]
variables [FlipHorizontal {type=boolean}]
widgets {
 info {inheritedWidgetCount=2, parentClass=/Script/FortniteUI.FortSidebarOnboardTooltipWidget, rootWidgetClass=&Overlay, widgetCount=20}
 widgets [
  Overlay_1 {widgetClassPath=&Overlay, widget=Overlay_1}
  SizeBox_187 {
   widgetClassPath=&SizeBox
   parent=Overlay_1
   widget=SizeBox_187
   properties {bOverride_MaxDesiredWidth=true, maxDesiredWidth=400}
   slot {class=OverlaySlot, properties {horizontalAlignment=HAlign_Fill, verticalAlignment=VAlign_Fill}}
  }
  HorizontalBox_0 {
   widgetClassPath=&HorizontalBox
   parent=SizeBox_187
   widget=HorizontalBox_0
   slot {class=SizeBoxSlot, properties {horizontalAlignment=HAlign_Left}}
  }
  SizeBox_0 {
   widgetClassPath=&SizeBox
   parent=HorizontalBox_0
   widget=SizeBox_0
   properties {bOverride_WidthOverride=true, widthOverride=550}
   slot {class=HorizontalBoxSlot, properties {horizontalAlignment=HAlign_Left}}
  }
  Overlay_79 {widgetClassPath=&Overlay, parent=SizeBox_0, widget=Overlay_79, properties {visibility=Visible}, slot {class=SizeBoxSlot}}
  SizeBox_170 {
   widgetClassPath=&SizeBox
   parent=HorizontalBox_0
   widget=SizeBox_170
   properties {bOverride_WidthOverride=true, minDesiredWidth=77, widthOverride=96}
   slot {class=HorizontalBoxSlot, properties {horizontalAlignment=HAlign_Right}}
  }
  Overlay_120 {widgetClassPath=&Overlay, parent=SizeBox_170, widget=Overlay_120, slot {class=SizeBoxSlot}}
  SizeBox_9 {
   widgetClassPath=&SizeBox
   parent=Overlay_1
   widget=SizeBox_9
   properties {bOverride_MinDesiredWidth=true, bOverride_WidthOverride=true, heightOverride=76, minDesiredWidth=200, widthOverride=550}
   slot {class=OverlaySlot, properties {horizontalAlignment=HAlign_Fill, verticalAlignment=VAlign_Center, padding=<ltrb 77 0 0 0>}}
  }
  Overlay_166 {widgetClassPath=&Overlay, bIsVariable=true, parent=SizeBox_9, widget=Overlay_166, slot {class=SizeBoxSlot}}
  ImageBurst {
   widgetClassPath=&Image
   bIsVariable=true
   parent=Overlay_166
   widget=ImageBurst
   properties {
    visibility=SelfHitTestInvisible
    brush {
     bIsDynamicallyLoaded=false
     drawAs=Box
     imageType=NoImage
     mirroring=NoMirror
     resourceName=None
     resourceObject=/FortUILibrary/Foundation/Art_ToolKit/AngledBoxes/M_UI_AngledBox_Tuto_burst
     tiling=NoTile
     imageSize=<90 90>
     margin=<ltrb 0.5 0 0.5 0>
     outlineSettings {
      bUseBrushTransparency=false
      roundingType=HalfHeightRadius
      width=0
      color {colorUseRule=UseColor_Specified, specifiedColor=<rgba 0 0 0 0>}
      cornerRadii=<0 0 0 0>
     }
     tintColor {colorUseRule=UseColor_Specified, specifiedColor=<rgba 1 1 1 1>}
     uVRegion {bIsValid=false, max=<0 0>, min=<0 0>}
    }
   }
   slot {class=OverlaySlot, properties {horizontalAlignment=HAlign_Fill, verticalAlignment=VAlign_Fill}}
  }
  Border_Shadow {
   widgetClassPath=&Border
   parent=Overlay_166
   widget=Border_Shadow
   properties {
    verticalAlignment=VAlign_Center
    background {
     bIsDynamicallyLoaded=false
     drawAs=Box
     imageType=NoImage
     mirroring=NoMirror
     resourceName=None
     resourceObject=/FortUILibrary/Foundation/Art_ToolKit/AngledBoxes/M_UI_AngledBox_Tuto
     tiling=NoTile
     imageSize=<90 90>
     margin=<ltrb 0.5 0 0.5 0>
     outlineSettings {
      bUseBrushTransparency=false
      roundingType=HalfHeightRadius
      width=0
      color {colorUseRule=UseColor_Specified, specifiedColor=<rgba 0 0 0 0>}
      cornerRadii=<0 0 0 0>
     }
     tintColor {colorUseRule=UseColor_Specified, specifiedColor=<rgba 1 1 1 1>}
     uVRegion {bIsValid=false, max=<0 0>, min=<0 0>}
    }
    brushColor=<rgba 0 0 0 0.5>
    padding=<ltrb 0 0 0 0>
    renderTransform {angle=0, scale=<1 1>, shear=<0 0>, translation=<8 10>}
   }
   slot {class=OverlaySlot, properties {horizontalAlignment=HAlign_Fill, verticalAlignment=VAlign_Fill}}
  }
  Border {
   widgetClassPath=&Border
   parent=Overlay_166
   widget=Border
   properties {
    verticalAlignment=VAlign_Center
    background {
     bIsDynamicallyLoaded=false
     drawAs=Box
     imageType=NoImage
     mirroring=NoMirror
     resourceName=None
     resourceObject=/FortUILibrary/Foundation/Art_ToolKit/AngledBoxes/M_UI_AngledBox_Tuto_Sidebar
     tiling=NoTile
     imageSize=<90 90>
     margin=<ltrb 0.5 0 0.5 0>
     outlineSettings {
      bUseBrushTransparency=false
      roundingType=HalfHeightRadius
      width=0
      color {colorUseRule=UseColor_Specified, specifiedColor=<rgba 0 0 0 0>}
      cornerRadii=<0 0 0 0>
     }
     tintColor {colorUseRule=UseColor_Specified, specifiedColor=<rgba 1 1 1 1>}
     uVRegion {bIsValid=false, max=<0 0>, min=<0 0>}
    }
    padding=<ltrb 0 0 0 0>
   }
   slot {class=OverlaySlot, properties {horizontalAlignment=HAlign_Fill, verticalAlignment=VAlign_Fill}}
  }
  VerticalBox_85 {
   widgetClassPath=&VerticalBox
   parent=Overlay_166
   widget=VerticalBox_85
   slot {class=OverlaySlot, properties {horizontalAlignment=HAlign_Fill, verticalAlignment=VAlign_Fill, padding=<ltrb 48 40 40 30>}}
  }
  CommonRichTextBlock_106 {
   widgetClassPath=&CommonRichTextBlock
   bIsVariable=true
   parent=VerticalBox_85
   widget=CommonRichTextBlock_106
   properties {
    autoWrapText=true
    bOverrideDefaultStyle=true
    text="Show Your Appreciation!"
    textTransformPolicy=ToUpper
    visibility=Collapsed
    defaultTextStyleOverride {
     overflowPolicy=Clip
     transformPolicy=None
     colorAndOpacity {colorUseRule=UseColor_Specified, specifiedColor=<rgba 1 1 1 1>}
     font {
      bForceMonospaced=false
      bMaterialIsStencil=false
      fontMaterial=null
      fontObject=&Burbank
      letterSpacing=0
      monospacedWidth=1
      size=24
      skewAmount=0
      typefaceFontName=BurbankRegular-Black
      outlineSettings {
       bApplyOutlineToDropShadows=false
       bMiteredCorners=false
       bSeparateFillAlpha=false
       outlineMaterial=null
       outlineSize=0
       outlineColor=<rgba 0 0 0 1>
      }
     }
     highlightColor {colorUseRule=UseColor_Specified, specifiedColor=<rgba 0 0 0 1>}
     highlightShape {
      bIsDynamicallyLoaded=false
      drawAs=Image
      imageType=NoImage
      mirroring=NoMirror
      resourceName=None
      resourceObject=null
      tiling=NoTile
      imageSize=<32 32>
      margin=<ltrb 0 0 0 0>
      outlineSettings {
       bUseBrushTransparency=false
       roundingType=HalfHeightRadius
       width=0
       color {colorUseRule=UseColor_Specified, specifiedColor=<rgba 0 0 0 0>}
       cornerRadii=<0 0 0 0>
      }
      tintColor {colorUseRule=UseColor_Specified, specifiedColor=<rgba 1 1 1 1>}
      uVRegion {bIsValid=false, max=<0 0>, min=<0 0>}
     }
     selectedBackgroundColor {colorUseRule=UseColor_Foreground, specifiedColor=<rgba 1 0 1 1>}
     shadowColorAndOpacity=<rgba 0 0 0 1>
     shadowOffset=<0 0>
     strikeBrush {
      bIsDynamicallyLoaded=false
      drawAs=Image
      imageType=NoImage
      mirroring=NoMirror
      resourceName=None
      resourceObject=null
      tiling=NoTile
      imageSize=<32 32>
      margin=<ltrb 0 0 0 0>
      outlineSettings {
       bUseBrushTransparency=false
       roundingType=HalfHeightRadius
       width=0
       color {colorUseRule=UseColor_Specified, specifiedColor=<rgba 0 0 0 0>}
       cornerRadii=<0 0 0 0>
      }
      tintColor {colorUseRule=UseColor_Specified, specifiedColor=<rgba 1 1 1 1>}
      uVRegion {bIsValid=false, max=<0 0>, min=<0 0>}
     }
     underlineBrush {
      bIsDynamicallyLoaded=false
      drawAs=Image
      imageType=NoImage
      mirroring=NoMirror
      resourceName=None
      resourceObject=null
      tiling=NoTile
      imageSize=<32 32>
      margin=<ltrb 0 0 0 0>
      outlineSettings {
       bUseBrushTransparency=false
       roundingType=HalfHeightRadius
       width=0
       color {colorUseRule=UseColor_Specified, specifiedColor=<rgba 0 0 0 0>}
       cornerRadii=<0 0 0 0>
      }
      tintColor {colorUseRule=UseColor_Specified, specifiedColor=<rgba 1 1 1 1>}
      uVRegion {bIsValid=false, max=<0 0>, min=<0 0>}
     }
    }
   }
   slot {class=VerticalBoxSlot}
  }
  CommonRichTextBlock_140 {
   widgetClassPath=&CommonRichTextBlock
   bIsVariable=true
   parent=VerticalBox_85
   widget=CommonRichTextBlock_140
   properties {
    autoWrapText=true
    bOverrideDefaultStyle=true
    mobileTextBlockScale=1.5
    text=Reach out and invite eligible returning friends or new players to participate in Reboot Rally
    defaultTextStyleOverride {
     overflowPolicy=Clip
     transformPolicy=None
     colorAndOpacity {colorUseRule=UseColor_Specified, specifiedColor=<rgba 0.4233 0.7529 1 1>}
     font {
      bForceMonospaced=false
      bMaterialIsStencil=false
      fontMaterial=null
      fontObject=&BurbankSmall
      letterSpacing=0
      monospacedWidth=1
      size=16
      skewAmount=0
      typefaceFontName=Bold
      outlineSettings {
       bApplyOutlineToDropShadows=false
       bMiteredCorners=false
       bSeparateFillAlpha=false
       outlineMaterial=null
       outlineSize=0
       outlineColor=<rgba 0 0 0 1>
      }
     }
     highlightColor {colorUseRule=UseColor_Specified, specifiedColor=<rgba 0 0 0 1>}
     highlightShape {
      bIsDynamicallyLoaded=false
      drawAs=Image
      imageType=NoImage
      mirroring=NoMirror
      resourceName=None
      resourceObject=null
      tiling=NoTile
      imageSize=<32 32>
      margin=<ltrb 0 0 0 0>
      outlineSettings {
       bUseBrushTransparency=false
       roundingType=HalfHeightRadius
       width=0
       color {colorUseRule=UseColor_Specified, specifiedColor=<rgba 0 0 0 0>}
       cornerRadii=<0 0 0 0>
      }
      tintColor {colorUseRule=UseColor_Specified, specifiedColor=<rgba 1 1 1 1>}
      uVRegion {bIsValid=false, max=<0 0>, min=<0 0>}
     }
     selectedBackgroundColor {colorUseRule=UseColor_Foreground, specifiedColor=<rgba 1 0 1 1>}
     shadowColorAndOpacity=<rgba 0 0 0 1>
     shadowOffset=<0 0>
     strikeBrush {
      bIsDynamicallyLoaded=false
      drawAs=Image
      imageType=NoImage
      mirroring=NoMirror
      resourceName=None
      resourceObject=null
      tiling=NoTile
      imageSize=<32 32>
      margin=<ltrb 0 0 0 0>
      outlineSettings {
       bUseBrushTransparency=false
       roundingType=HalfHeightRadius
       width=0
       color {colorUseRule=UseColor_Specified, specifiedColor=<rgba 0 0 0 0>}
       cornerRadii=<0 0 0 0>
      }
      tintColor {colorUseRule=UseColor_Specified, specifiedColor=<rgba 1 1 1 1>}
      uVRegion {bIsValid=false, max=<0 0>, min=<0 0>}
     }
     underlineBrush {
      bIsDynamicallyLoaded=false
      drawAs=Image
      imageType=NoImage
      mirroring=NoMirror
      resourceName=None
      resourceObject=null
      tiling=NoTile
      imageSize=<32 32>
      margin=<ltrb 0 0 0 0>
      outlineSettings {
       bUseBrushTransparency=false
       roundingType=HalfHeightRadius
       width=0
       color {colorUseRule=UseColor_Specified, specifiedColor=<rgba 0 0 0 0>}
       cornerRadii=<0 0 0 0>
      }
      tintColor {colorUseRule=UseColor_Specified, specifiedColor=<rgba 1 1 1 1>}
      uVRegion {bIsValid=false, max=<0 0>, min=<0 0>}
     }
    }
   }
   slot {class=VerticalBoxSlot, properties {verticalAlignment=VAlign_Center}}
  }
  ScaleBox_mobile {
   widgetClassPath=&ScaleBox
   bIsVariable=true
   parent=VerticalBox_85
   widget=ScaleBox_mobile
   properties {stretch=UserSpecified}
   slot {class=VerticalBoxSlot, properties {horizontalAlignment=HAlign_Right, padding=<ltrb 0 0 -5 0>}}
  }
  SizeBox_mobile {
   widgetClassPath=&SizeBox
   bIsVariable=true
   parent=ScaleBox_mobile
   widget=SizeBox_mobile
   properties {heightOverride=60}
   slot {class=ScaleBoxSlot}
  }
  Dismiss_Button {
   widgetClassPath=/Game/UI/Sidebar/OnboardingTextInputActionButton
   bInherited=true
   bIsVariable=true
   parent=SizeBox_mobile
   widget=Dismiss_Button
   properties {
    bShouldUseFallbackDefaultInputAction=false
    buttonText=dismiss
    onButtonBaseClicked=(&self:WidgetTree.Dismiss_Button.HandleOnButtonClicked)
    onButtonBaseDoubleClicked=(&self:WidgetTree.Dismiss_Button.HandleOnButtonDoubleClicked)
    onButtonBaseHovered=(&self:WidgetTree.Dismiss_Button.HandleOnButtonHovered)
    onButtonBaseUnhovered=(&self:WidgetTree.Dismiss_Button.HandleOnButtonUnhovered)
    onSelectedChangedBase=(&self:WidgetTree.Dismiss_Button.HandleOnSelectedChanged)
    style=/FortUILibrary/Foundation/Buttons/ButtonStyles/ButtonStyle-Clear.ButtonStyle-Clear_C
    triggeringInputAction {dataTable=&InputActionDataTable, rowName=GenericBack}
   }
   slot {class=SizeBoxSlot, properties {horizontalAlignment=HAlign_Right, padding=<ltrb 0 0 10 0>}}
  }
  Image_Arrow {
   widgetClassPath=&Image
   bIsVariable=true
   parent=Overlay_166
   widget=Image_Arrow
   properties {
    brush {
     bIsDynamicallyLoaded=false
     drawAs=Image
     imageType=NoImage
     mirroring=NoMirror
     resourceName=None
     resourceObject=/Game/Athena/HUD/Quests/Art/T_UI_Quest_Arrow
     tiling=NoTile
     imageSize=<60 60>
     margin=<ltrb 0 0 0 0>
     outlineSettings {
      bUseBrushTransparency=false
      roundingType=HalfHeightRadius
      width=0
      color {colorUseRule=UseColor_Specified, specifiedColor=<rgba 0 0 0 0>}
      cornerRadii=<0 0 0 0>
     }
     tintColor {colorUseRule=UseColor_Specified, specifiedColor=<rgba 1 1 1 1>}
     uVRegion {bIsValid=false, max=<0 0>, min=<0 0>}
    }
    renderTransform {angle=-90, scale=<1 1>, shear=<0 0>, translation=<-47 0>}
   }
   slot {class=OverlaySlot, properties {verticalAlignment=VAlign_Center, padding=<ltrb 0 0 0 2>}}
  }
  OtherAction_Button {widgetClassPath=&CommonButtonLegacy, bInherited=true, widget=null}
 ]
}
