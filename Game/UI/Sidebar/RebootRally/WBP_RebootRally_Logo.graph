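# Widget Blueprint that displays the Reboot Rally logo lockup for the sidebar UI.
# A root Overlay stacks a glitch panel material, a faint blue radial-gradient glow and the logo material, all centered horizontally.
# A padded nested Overlay places an uppercase, bold italic REBOOT RALLY CommonTextBlock in the lower part.
# The widget is layout only: its event stubs are unconnected and it declares no variables in the graph.
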WidgetBlueprint /Game/UI/Sidebar/RebootRally/WBP_RebootRally_Logo
graphs [EventGraph {nodes [
   Event PreConstruct()
   Event Construct()
   Event Tick()
  ]}]
variables []
widgets {
 info {inheritedWidgetCount=0, parentClass=&UserWidget, rootWidgetClass=&Overlay, widgetCount=6}
 widgets [
  Overlay_0 {widgetClassPath=&Overlay, widget=Overlay_0}
  Image_GlitchPanel {
   widgetClassPath=&Image
   bIsVariable=true
   parent=Overlay_0
   widget=Image_GlitchPanel
   properties {
    visibility=SelfHitTestInvisible
    brush {
     bIsDynamicallyLoaded=false
     drawAs=Image
     imageType=NoImage
     mirroring=NoMirror
     resourceName=None
     resourceObject=/Game/Athena/UI/Frontend/Spatial/Art/InteractionIndicators/M_UI_POI_Glitch_Text_Panel
     tiling=NoTile
     imageSize=<128 120>
     margin=<ltrb 0 0 0 0>
     outlineSettings {
      bUseBrushTransparency=false
      roundingType=HalfHeightRadius
      width=0
      color {colorUseRule=UseColor_Specified, specifiedColor=<rgba 0 0 0 0>}
      cornerRadii=<0 0 0 0>
     }
     tintColor {colorUseRule=UseColor_Specified, specifiedColor=<rgba 1 1 1 1>}
     uVRegion {bIsValid=false, max=<0 0>, min=<0 0>}
    }
    renderTransform {angle=0, scale=<1.2 1.2>, shear=<0 0>, translation=<0 -37>}
   }
   slot {class=OverlaySlot, properties {horizontalAlignment=HAlign_Center}}
  }
  Image_TextGlow01 {
   widgetClassPath=&Image
   bIsVariable=true
   parent=Overlay_0
   widget=Image_TextGlow01
   properties {
    brush {
     bIsDynamicallyLoaded=false
     drawAs=Image
     imageType=NoImage
     mirroring=NoMirror
     resourceName=None
     resourceObject=/FortUILibrary/MaterialInstances/MI_UI_RadialGradientTransparent
     tiling=NoTile
     imageSize=<160 72>
     margin=<ltrb 0 0 0 0>
     outlineSettings {
      bUseBrushTransparency=false
      roundingType=HalfHeightRadius
      width=0
      color {colorUseRule=UseColor_Specified, specifiedColor=<rgba 0 0 0 0>}
      cornerRadii=<0 0 0 0>
     }
     tintColor {colorUseRule=UseColor_Specified, specifiedColor=<rgba 0 0.3818 0.9911 1>}
     uVRegion {bIsValid=false, max=<0 0>, min=<0 0>}
    }
    colorAndOpacity=<rgba 1 1 1 0.3>
    renderTransform {angle=0, scale=<1 1>, shear=<0 0>, translation=<0 -12>}
   }
   slot {class=OverlaySlot, properties {horizontalAlignment=HAlign_Center}}
  }
  Image_RebootRallyLogo {
   widgetClassPath=&Image
   bIsVariable=true
   parent=Overlay_0
   widget=Image_RebootRallyLogo
   properties {brush {
     bIsDynamicallyLoaded=false
     drawAs=Image
     imageType=NoImage
     mirroring=NoMirror
     resourceName=None
     resourceObject=/Game/UI/Sidebar/RebootRally/Materials/M_UI_RebootRallyLogo
     tiling=NoTile
     imageSize=<112 112>
     margin=<ltrb 0 0 0 0>
     outlineSettings {
      bUseBrushTransparency=false
      roundingType=HalfHeightRadius
      width=0
      color {colorUseRule=UseColor_Specified, specifiedColor=<rgba 0 0 0 0>}
      cornerRadii=<0 0 0 0>
     }
     tintColor {colorUseRule=UseColor_Specified, specifiedColor=<rgba 1 1 1 1>}
     uVRegion {bIsValid=false, max=<0 0>, min=<0 0>}
    }}
   slot {class=OverlaySlot, properties {horizontalAlignment=HAlign_Center}}
  }
  Overlay_REBOOT {
   widgetClassPath=&Overlay
   parent=Overlay_0
   widget=Overlay_REBOOT
   slot {class=OverlaySlot, properties {horizontalAlignment=HAlign_Fill, padding=<ltrb 0 86 0 0>}}
  }
  CommonTextBlock {
   widgetClassPath=&CommonTextBlock
   parent=Overlay_REBOOT
   widget=CommonTextBlock
   properties {
    autoWrapText=true
    bIsScrollingEnabled=false
    justification=Center
    lineHeightPercentage=0.7
    text=REBOOT RALLY
    textTransformPolicy=ToUpper
    wrapTextAt=420
    font {
     bForceMonospaced=false
     bMaterialIsStencil=false
     fontMaterial=null
     fontObject=&HeadingNow
     letterSpacing=-40
     monospacedWidth=1
     size=33
     skewAmount=0
     typefaceFontName=86 Bold Italic
     outlineSettings {
      bApplyOutlineToDropShadows=false
      bMiteredCorners=false
      bSeparateFillAlpha=false
      outlineMaterial=null
      outlineSize=0
      outlineColor=<rgba 0 0 0 1>
     }
    }
    margin=<ltrb 0 -14 0 8>
    shadowOffset=<0 0>
   }
   slot {class=OverlaySlot, properties {horizontalAlignment=HAlign_Fill, verticalAlignment=VAlign_Bottom}}
  }
 ]
}
