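# Header widget for the Reboot Rally sidebar panel, showing rewards art, logo, tagline, details text and a View Quests button over a cloud background.
# It extends FortSidebarPanelRebootHeader, whose eligibility event swaps the rich text between eligible and ineligible copy, styled with inline icons and highlights.
# The parent's visibility event shows or hides the quests button, which the widget exposes through OnGetQuestsButton and whose clicks go to HandleQuestsButtonClicked.
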WidgetBlueprint /FortniteCoreUI/UI/Sidebar/RebootRally/WBP_PanelRebootRally_Header
animations [Default {binding_count=0, length_seconds=5, object=&self:Default, sequence {bindings []}}]
graphs [
 EventGraph {nodes [
   "On Clicked (WBP_UIKit_Button_Regular)"!() -> (OutputDelegate, then, Button:Common Button Base Object Reference) {is=title K2Node_ComponentBoundEvent}
   > HandleQuestsButtonClicked(self:Fort Sidebar Panel Reboot Header Object Reference)
   n1 = Event OnRebootRallyEligibilityUpdated!() -> (OutputDelegate, then, bEligible:bool) {is=title K2Node_Event}
   > SetText(
    self=Get RichText_RebootRallyDetails() -> (RichText_RebootRallyDetails:Common Rich Text Block Object Reference)
    InText=Select(
     Option 0:"Text (by ref)"=Get Text_Ineligible() -> (Text_Ineligible:Text)
     Option 1:"Text (by ref)"=Get Text_Eligible() -> (Text_Eligible:Text)
     Index=@n1.bEligible
    ) -> (ReturnValue:"Text (by ref)")
   )
   n2 = Event OnUpdateQuestsButtonVisibility!() -> (OutputDelegate, then, bVisible:bool) {is=title K2Node_Event}
   > SetVisibility(
    self=Get WBP_UIKit_Button_Regular() -> (WBP_UIKit_Button_Regular:WBP UIKit Button Regular Object Reference)
    InVisibility=Select(Option 0:ESlateVisibility Enum=Hidden, Option 1:ESlateVisibility Enum=Visible, Index=@n2.bVisible) -> (ReturnValue:ESlateVisibility Enum)
   )
   Event Construct()
  ]}
 OnGetQuestsButton {nodes [
   OnGetQuestsButton() {is=title K2Node_FunctionEntry}
   > Return Node!(execute, ReturnValue:Common Button Base Object Reference=Get WBP_UIKit_Button_Regular())
  ]}
]
variables [
 Text_Eligible {
  type=string
  default="You've been <img id=\"RebootRally\"/> <HighlightBlue14>rallied!</> Play with friends to earn cosmetic rewards together! You can also rally <img id=\"RebootRally\"/> <HighlightBlue14>eligible</> returning friends or new players."
 }
 Text_Ineligible {
  type=string
  default="Rally <img id=\"RebootRally\"/> <HighlightBlue>eligible</> returning friends or new players to play and earn cosmetic rewards together."
 }
]
widgets {
 info {
  inheritedWidgetCount=0
  parentClass=/Script/FortniteUI.FortSidebarPanelRebootHeader
  rootWidgetClass=&CommonBorder
  widgetCount=11
 }
 widgets [
  Border_Background {
   widgetClassPath=&CommonBorder
   widget=Border_Background
   properties {
    background {
     bIsDynamicallyLoaded=false
     drawAs=Image
     imageType=FullColor
     mirroring=NoMirror
     resourceName=None
     resourceObject=/Game/UI/Sidebar/RebootRally/Materials/MI_UI_RebootRally_BGClouds_S32_Header
     tiling=NoTile
     imageSize=<0 0>
     margin=<ltrb 0 0 0 0>
     outlineSettings {
      bUseBrushTransparency=false
      roundingType=HalfHeightRadius
      width=0
      color {colorUseRule=UseColor_Specified, specifiedColor=<rgba 0 0 0 0>}
      cornerRadii=<0 0 0 0>
     }
     tintColor {colorUseRule=UseColor_Specified, specifiedColor=<rgba 1 1 1 1>}
     uVRegion {bIsValid=false, max=<0 0>, min=<0 0>}
    }
    padding=<ltrb 0 0 0 0>
   }
  }
  GridPanel_255 {
   widgetClassPath=&GridPanel
   parent=Border_Background
   widget=GridPanel_255
   properties {columnFill [1], rowFill [0,0,0]}
   slot {class=BorderSlot, properties {padding=<ltrb 0 0 0 0>}}
  }
  WBP_CaptureForPostBufferUpdate {
   widgetClassPath=&WBP_CaptureForPostBufferUpdate
   bIsVariable=true
   parent=GridPanel_255
   slot {class=GridSlot, properties {layer=-1, rowSpan=3}}
  }
  Overlay_RebootRallyRewards {
   widgetClassPath=&Overlay
   parent=GridPanel_255
   widget=Overlay_RebootRallyRewards
   slot {class=GridSlot, properties {horizontalAlignment=HAlign_Center, verticalAlignment=VAlign_Top, padding=<ltrb 0 0 0 12>}}
  }
  Spacer_133 {
   widgetClassPath=&Spacer
   parent=Overlay_RebootRallyRewards
   widget=Spacer_133
   properties {size=<1024 460>}
   slot {class=OverlaySlot, properties {horizontalAlignment=HAlign_Fill, verticalAlignment=VAlign_Fill}}
  }
  Image_Rewards {
   widgetClassPath=&Image
   bIsVariable=true
   parent=Overlay_RebootRallyRewards
   widget=Image_Rewards
   properties {
    visibility=SelfHitTestInvisible
    brush {
     bIsDynamicallyLoaded=false
     drawAs=Image
     imageType=NoImage
     mirroring=NoMirror
     resourceName=None
     resourceObject=/Game/UI/Sidebar/RebootRally/Materials/MI_UI_Reward_Composition_MainHeader
     tiling=NoTile
     imageSize=<32 32>
     margin=<ltrb 0 0 0 0>
     outlineSettings {
      bUseBrushTransparency=false
      roundingType=HalfHeightRadius
      width=0
      color {colorUseRule=UseColor_Specified, specifiedColor=<rgba 0 0 0 0>}
      cornerRadii=<0 0 0 0>
     }
     tintColor {colorUseRule=UseColor_Specified, specifiedColor=<rgba 1 1 1 1>}
     uVRegion {bIsValid=false, max=<0 0>, min=<0 0>}
    }
   }
   slot {class=OverlaySlot, properties {horizontalAlignment=HAlign_Fill, verticalAlignment=VAlign_Fill, padding=<ltrb 0 24 0 0>}}
  }
  VBox_Logo {
   widgetClassPath=&VerticalBox
   parent=Overlay_RebootRallyRewards
   widget=VBox_Logo
   slot {class=OverlaySlot, properties {horizontalAlignment=HAlign_Fill, verticalAlignment=VAlign_Bottom}}
  }
  WBP_RebootRally_Logo {
   widgetClassPath=/Game/UI/Sidebar/RebootRally/WBP_RebootRally_Logo
   bIsVariable=true
   parent=VBox_Logo
   widget=WBP_RebootRally_Logo
   slot {class=VerticalBoxSlot, properties {horizontalAlignment=HAlign_Center, verticalAlignment=VAlign_Bottom}}
  }
  Text_RebootRallyHeader {
   widgetClassPath=&CommonTextBlock
   parent=VBox_Logo
   widget=Text_RebootRallyHeader
   properties {
    justification=Center
    mobileFontSizeMultiplier=1.2
    text="Play, Reboot, Reward"
    textTransformPolicy=ToUpper
    colorAndOpacity {colorUseRule=UseColor_Specified, specifiedColor=<rgba 0 0 0 0.75>}
    font {
     bForceMonospaced=false
     bMaterialIsStencil=false
     fontMaterial=null
     fontObject=&HeadingNow
     letterSpacing=0
     monospacedWidth=1
     size=12.75
     skewAmount=0
     typefaceFontName=75 Medium
     outlineSettings {
      bApplyOutlineToDropShadows=false
      bMiteredCorners=false
      bSeparateFillAlpha=false
      outlineMaterial=null
      outlineSize=0
      outlineColor=<rgba 0 0 0 1>
     }
    }
    shadowOffset=<0 0>
   }
   slot {class=VerticalBoxSlot, properties {horizontalAlignment=HAlign_Center, verticalAlignment=VAlign_Bottom, padding=<ltrb 0 12 0 0>}}
  }
  RichText_RebootRallyDetails {
   widgetClassPath=&CommonRichTextBlock
   bIsVariable=true
   parent=GridPanel_255
   widget=RichText_RebootRallyDetails
   properties {
    autoWrapText=true
    bOverrideDefaultStyle=true
    justification=Center
    mobileTextBlockScale=1.15
    text="You've been <img id=\"RebootRally\"/> <HighlightBlue14>rallied!</> Play with friends to earn cosmetic rewards together! You can also rally <img id=\"RebootRally\"/> <HighlightBlue14>eligible</> returning friends or new players."
    textStyleSet=/Game/UI/Sidebar/RebootRally/DT_UI_RebootRallyRichTextStyles
    decoratorClasses [/Game/UI/Sidebar/RebootRally/BP_UI_RebootRallyRichImageDecorator.BP_UI_RebootRallyRichImageDecorator_C]
    defaultTextStyleOverride {
     overflowPolicy=Clip
     transformPolicy=None
     colorAndOpacity {colorUseRule=UseColor_Specified, specifiedColor=<rgba 0 0 0 1>}
     font {
      bForceMonospaced=false
      bMaterialIsStencil=false
      fontMaterial=null
      fontObject=&HeadingNow
      letterSpacing=0
      monospacedWidth=1
      size=12
      skewAmount=0
      typefaceFontName=74 Regular
      outlineSettings {
       bApplyOutlineToDropShadows=false
       bMiteredCorners=false
       bSeparateFillAlpha=false
       outlineMaterial=null
       outlineSize=0
       outlineColor=<rgba 0 0 0 1>
      }
     }
     highlightColor {colorUseRule=UseColor_Specified, specifiedColor=<rgba 0 0 0 1>}
     highlightShape {
      bIsDynamicallyLoaded=false
      drawAs=Image
      imageType=NoImage
      mirroring=NoMirror
      resourceName=None
      resourceObject=null
      tiling=NoTile
      imageSize=<32 32>
      margin=<ltrb 0 0 0 0>
      outlineSettings {
       bUseBrushTransparency=false
       roundingType=HalfHeightRadius
       width=0
       color {colorUseRule=UseColor_Specified, specifiedColor=<rgba 0 0 0 0>}
       cornerRadii=<0 0 0 0>
      }
      tintColor {colorUseRule=UseColor_Specified, specifiedColor=<rgba 1 1 1 1>}
      uVRegion {bIsValid=false, max=<0 0>, min=<0 0>}
     }
     selectedBackgroundColor {colorUseRule=UseColor_Foreground, specifiedColor=<rgba 1 0 1 1>}
     shadowColorAndOpacity=<rgba 0 0 0 1>
     shadowOffset=<0 0>
     strikeBrush {
      bIsDynamicallyLoaded=false
      drawAs=Image
      imageType=NoImage
      mirroring=NoMirror
      resourceName=None
      resourceObject=null
      tiling=NoTile
      imageSize=<32 32>
      margin=<ltrb 0 0 0 0>
      outlineSettings {
       bUseBrushTransparency=false
       roundingType=HalfHeightRadius
       width=0
       color {colorUseRule=UseColor_Specified, specifiedColor=<rgba 0 0 0 0>}
       cornerRadii=<0 0 0 0>
      }
      tintColor {colorUseRule=UseColor_Specified, specifiedColor=<rgba 1 1 1 1>}
      uVRegion {bIsValid=false, max=<0 0>, min=<0 0>}
     }
     underlineBrush {
      bIsDynamicallyLoaded=false
      drawAs=Image
      imageType=NoImage
      mirroring=NoMirror
      resourceName=None
      resourceObject=null
      tiling=NoTile
      imageSize=<32 32>
      margin=<ltrb 0 0 0 0>
      outlineSettings {
       bUseBrushTransparency=false
       roundingType=HalfHeightRadius
       width=0
       color {colorUseRule=UseColor_Specified, specifiedColor=<rgba 0 0 0 0>}
       cornerRadii=<0 0 0 0>
      }
      tintColor {colorUseRule=UseColor_Specified, specifiedColor=<rgba 1 1 1 1>}
      uVRegion {bIsValid=false, max=<0 0>, min=<0 0>}
     }
    }
   }
   slot {class=GridSlot, properties {row=1, verticalAlignment=VAlign_Top, padding=<ltrb 72 16 72 0>}}
  }
  WBP_UIKit_Button_Regular {
   widgetClassPath=&WBP_UIKit_Button_Regular
   bIsVariable=true
   parent=GridPanel_255
   widget=WBP_UIKit_Button_Regular
   properties {text=View Quests}
   slot {class=GridSlot, properties {row=2, padding=<ltrb 64 24 64 64>}}
  }
 ]
}
