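# Sidebar list entry widget that shows a "Looking For Party" label, a select-action icon and a text toggle.
# The native SidebarLFGToggleListEntryWidget parent raises OnLfgSettingChanged, which the graph passes to the toggle's state.
# Hover and unhover animate the background material highlight, text color, padding and icon opacity, with unhover at double speed.
# The input icon's dynamic material is switched to dark without outline, and mobile gets a taller minimum height.
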WidgetBlueprint /FortniteCoreUI/UI/Sidebar/Tags/SidebarLFGToggleListEntry
animations [
 Hovered {
  binding_count=5
  length_seconds=0.2
  object=&self:Hovered
  sequence {bindings [
    HorizontalBoxSlot_3 {
     id=#1
     parent_id=#2
     possessed_class=HorizontalBoxSlot
     tracks [{
       type=MovieSceneMarginTrack
       display_name=Padding
       sections [{
         type=MovieSceneMarginSection
         end_frame=4
         start_frame=0
         channels [
          Left {type=MovieSceneScriptingFloatChannel, keys [[0,20],[3,30]]}
          Top {type=MovieSceneScriptingFloatChannel}
          Right {type=MovieSceneScriptingFloatChannel}
          Bottom {type=MovieSceneScriptingFloatChannel}
         ]
        }]
      }]
    }
    CText_LFP {
     id=#2
     possessed_class=CommonTextBlock
     tracks [{
       type=MovieSceneColorTrack
       display_name=Color and Opacity
       sections [{
         type=MovieSceneColorSection
         end_frame=4
         start_frame=0
         channels [
          Color.R {type=MovieSceneScriptingFloatChannel, keys [[0,1],[2,0]]}
          Color.G {type=MovieSceneScriptingFloatChannel, keys [[0,1],[2,0]]}
          Color.B {type=MovieSceneScriptingFloatChannel, keys [[0,1],[2,0]]}
          Color.A {type=MovieSceneScriptingFloatChannel, const [1,2]}
         ]
        }]
      }]
    }
    InputActionWidget {
     id=#3
     possessed_class=CommonActionWidget
     tracks [{
       type=MovieSceneFloatTrack
       display_name=Render Opacity
       sections [{type=MovieSceneFloatSection, end_frame=4, start_frame=0, channels [None {type=MovieSceneScriptingFloatChannel, keys [[0,0],[2,1]]}]}]
      }]
    }
    Border_Background {
     id=#4
     possessed_class=Border
     tracks [{
       type=MovieSceneWidgetMaterialTrack
       sections [{
         type=MovieSceneParameterSection
         end_frame=4
         start_frame=0
         channels [AnimateHighlight {type=MovieSceneScriptingFloatChannel, keys [[0,0],[3,1],[4,1]]}]
        }]
      }]
    }
    BorderSlot_2 {
     id=#5
     parent_id=#3
     possessed_class=BorderSlot
     tracks [{
       type=MovieSceneMarginTrack
       display_name=Padding
       sections [{
         type=MovieSceneMarginSection
         end_frame=2
         start_frame=0
         channels [
          Left {type=MovieSceneScriptingFloatChannel}
          Top {type=MovieSceneScriptingFloatChannel}
          Right {type=MovieSceneScriptingFloatChannel, keys [[0,0],[2,10]]}
          Bottom {type=MovieSceneScriptingFloatChannel}
         ]
        }]
      }]
    }
   ]}
 }
 Unhovered {
  binding_count=5
  length_seconds=0.2
  object=&self:Unhovered
  sequence {bindings [
    BorderSlot_2 {
     id=#6
     parent_id=#7
     possessed_class=BorderSlot
     tracks [{
       type=MovieSceneMarginTrack
       display_name=Padding
       sections [{
         type=MovieSceneMarginSection
         end_frame=4
         start_frame=0
         channels [
          Left {type=MovieSceneScriptingFloatChannel}
          Top {type=MovieSceneScriptingFloatChannel}
          Right {type=MovieSceneScriptingFloatChannel, keys [[0,10],[2,10],[4,0]]}
          Bottom {type=MovieSceneScriptingFloatChannel}
         ]
        }]
      }]
    }
    CText_LFP {
     id=#8
     possessed_class=CommonTextBlock
     tracks [{
       type=MovieSceneColorTrack
       display_name=Color and Opacity
       sections [{
         type=MovieSceneColorSection
         end_frame=4
         start_frame=0
         channels [
          Color.R {type=MovieSceneScriptingFloatChannel, keys [[0,0],[2,0],[4,1]]}
          Color.G {type=MovieSceneScriptingFloatChannel, keys [[0,0],[2,0],[4,1]]}
          Color.B {type=MovieSceneScriptingFloatChannel, keys [[0,0],[2,0],[4,1]]}
          Color.A {type=MovieSceneScriptingFloatChannel, const [1,3]}
         ]
        }]
      }]
    }
    InputActionWidget {
     id=#7
     possessed_class=CommonActionWidget
     tracks [{
       type=MovieSceneFloatTrack
       display_name=Render Opacity
       sections [{type=MovieSceneFloatSection, end_frame=4, start_frame=0, channels [None {type=MovieSceneScriptingFloatChannel, keys [[2,1],[4,0]]}]}]
      }]
    }
    Border_Background {
     id=#9
     possessed_class=Border
     tracks [{
       type=MovieSceneWidgetMaterialTrack
       sections [{
         type=MovieSceneParameterSection
         end_frame=4
         start_frame=0
         channels [AnimateHighlight {type=MovieSceneScriptingFloatChannel, keys [[0,1],[1,1],[4,0]]}]
        }]
      }]
    }
    HorizontalBoxSlot_3 {
     id=#10
     parent_id=#8
     possessed_class=HorizontalBoxSlot
     tracks [{
       type=MovieSceneMarginTrack
       display_name=Padding
       sections [{
         type=MovieSceneMarginSection
         end_frame=4
         start_frame=0
         channels [
          Left {type=MovieSceneScriptingFloatChannel, keys [[0,30],[1,30],[4,20]]}
          Top {type=MovieSceneScriptingFloatChannel}
          Right {type=MovieSceneScriptingFloatChannel}
          Bottom {type=MovieSceneScriptingFloatChannel}
         ]
        }]
      }]
    }
   ]}
 }
]
graphs [EventGraph {nodes [
   "On Input Icon Updated (InputActionWidget)"() {is=title K2Node_ComponentBoundEvent}
   > n1 = GetIconDynamicMaterial!(execute, self:Common Action Widget Object Reference=Get InputActionWidget() -> (InputActionWidget:Common Action Widget Object Reference)) -> (then, ReturnValue:Material Instance Dynamic Object Reference)
   > SetScalarParameterValue(self=@n1, ParameterName=Use Dark, Value=1)
   > SetScalarParameterValue(self=@n1, ParameterName=Outline Opacity, Value=0)
   n2 = Event OnLfgSettingChanged!() -> (OutputDelegate, then, bIsLfgEnabled:bool) {is=title K2Node_Event}
   > SetToggleState!(
    execute
    self:Text Toggle Object Reference=Get SidebarSocialTagsLFPToggle() -> (SidebarSocialTagsLFPToggle:Text Toggle Object Reference)
    IsOn:bool=@n2.bIsLfgEnabled
   ) -> (then)
   Event On Hovered()
   > StopAllAnimations(self:User Widget Object Reference)
   > PlayAnimation(InAnimation=Get Hovered() -> (Hovered:Widget Animation Object Reference))
   Event On Unhovered()
   > StopAllAnimations()
   > PlayAnimation(InAnimation=Get Unhovered() -> (Unhovered:Widget Animation Object Reference), PlaybackSpeed=2)
   Event OnInitialized()
   > Branch(Condition=IsMobileGame())
   > SetMinDesiredHeight!(execute, self:Size Box Object Reference=Get SizeBox_0() -> (SizeBox_0:Size Box Object Reference), InMinDesiredHeight:float=92) -> (then)
  ]}]
variables [
 EncourageEpicFriend_HoveredColor {type=LinearColor, default=<rgba 1 1 1 1>}
 EncourageEpicFriend_IdleColor {type=LinearColor, default=<rgba 0.9558 1 0.0672 1>}
]
widgets {
 info {inheritedWidgetCount=2, parentClass=/Script/FortniteUI.SidebarLFGToggleListEntryWidget, rootWidgetClass=&SizeBox, widgetCount=8}
 widgets [
  SizeBox_0 {
   widgetClassPath=&SizeBox
   bIsVariable=true
   widget=SizeBox_0
   properties {bOverride_MinDesiredHeight=true, heightOverride=72, minDesiredHeight=72}
  }
  Border_Background {
   widgetClassPath=&Border
   bIsVariable=true
   parent=SizeBox_0
   widget=Border_Background
   properties {
    clipping=ClipToBounds
    visibility=SelfHitTestInvisible
    background {
     bIsDynamicallyLoaded=false
     drawAs=Image
     imageType=NoImage
     mirroring=NoMirror
     resourceName=None
     resourceObject=/Game/UI/Sidebar/Materials/M_UI_sidebar_Button_Selection_Inst
     tiling=NoTile
     imageSize=<32 32>
     margin=<ltrb 0 0 0 0>
     outlineSettings {
      bUseBrushTransparency=false
      roundingType=HalfHeightRadius
      width=0
      color {colorUseRule=UseColor_Specified, specifiedColor=<rgba 0 0 0 0>}
      cornerRadii=<0 0 0 0>
     }
     tintColor {colorUseRule=UseColor_Specified, specifiedColor=<rgba 1 1 1 1>}
     uVRegion {bIsValid=false, max=<0 0>, min=<0 0>}
    }
    padding=<ltrb 0 0 0 0>
   }
   slot {class=SizeBoxSlot}
  }
  Overlay_0 {
   widgetClassPath=&Overlay
   parent=Border_Background
   widget=Overlay_0
   slot {class=BorderSlot, properties {padding=<ltrb 0 0 0 0>}}
  }
  HorizontalBox_2 {
   widgetClassPath=&HorizontalBox
   parent=Overlay_0
   widget=HorizontalBox_2
   slot {class=OverlaySlot, properties {horizontalAlignment=HAlign_Fill, verticalAlignment=VAlign_Fill}}
  }
  CText_LFP {
   widgetClassPath=&CommonTextBlock
   parent=HorizontalBox_2
   widget=CText_LFP
   properties {
    autoWrapText=true
    mobileFontSizeMultiplier=1.25
    text=Looking For Party
    textTransformPolicy=ToUpper
    font {
     bForceMonospaced=false
     bMaterialIsStencil=false
     fontMaterial=null
     fontObject=&BurbankSmall
     letterSpacing=0
     monospacedWidth=1
     size=16
     skewAmount=0
     typefaceFontName=Black
     outlineSettings {
      bApplyOutlineToDropShadows=false
      bMiteredCorners=false
      bSeparateFillAlpha=false
      outlineMaterial=null
      outlineSize=0
      outlineColor=<rgba 0 0 0 1>
     }
    }
    shadowOffset=<0 0>
   }
   slot {class=HorizontalBoxSlot, properties {verticalAlignment=VAlign_Center, padding=<ltrb 20 0 10 0>, size {sizeRule=Fill, value=1}}}
  }
  InputActionWidget {
   widgetClassPath=&CommonActionWidget
   bInherited=true
   parent=HorizontalBox_2
   widget=InputActionWidget
   properties {renderOpacity=0, inputActions [{dataTable=&InputActionDataTable, rowName=Select}]}
   slot {class=HorizontalBoxSlot, properties {horizontalAlignment=HAlign_Right, verticalAlignment=VAlign_Center}}
  }
  SidebarSocialTagsLFPToggle {
   widgetClassPath=/Game/UI/DiscoveryShared/Status/TextToggle
   bIsVariable=true
   parent=HorizontalBox_2
   widget=SidebarSocialTagsLFPToggle
   slot {class=HorizontalBoxSlot, properties {horizontalAlignment=HAlign_Right, verticalAlignment=VAlign_Center, padding=<ltrb 0 0 20 0>}}
  }
  MenuAnchor_Actions {
   widgetClassPath=/Script/UMG.MenuAnchor
   bInherited=true
   bIsVariable=true
   parent=Overlay_0
   widget=MenuAnchor_Actions
   properties {menuClass=/Game/UI/SocialPanel/UserActionMenu.UserActionMenu_C, placement=MenuPlacement_MenuRight}
   slot {class=OverlaySlot, properties {horizontalAlignment=HAlign_Fill, verticalAlignment=VAlign_Fill}}
  }
 ]
}
