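# Collapsible category header row for the sidebar LFG tag filter list, showing category name, entry count, input prompts and an expand arrow.
# Hover plays a highlight animation and shows the focused input prompt; unhover or focus loss reverses the animation and hides that prompt.
# Expansion changes rotate the arrow by playing OnExpansion forward or reversed, with a sound.
# When bound to an already expanded item, it jumps that animation to its end and takes focus.
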WidgetBlueprint /FortniteCoreUI/UI/Sidebar/Tags/SidebarLfgFilterListHeaderEntry
animations [
 OnHover {
  binding_count=7
  display_label=""
  length_seconds=0.2
  object=&self:OnHover
  sequence {bindings [
    ExpandDirectionTick {
     id=#1
     possessed_class=Image
     tracks [
      {
       type=MovieScene2DTransformTrack
       sections [{
         type=MovieScene2DTransformSection
         channels [
          Translation.X {type=MovieSceneScriptingFloatChannel}
          Translation.Y {type=MovieSceneScriptingFloatChannel, keys [[0,0],[3,-6],[4,0]]}
          Angle {type=MovieSceneScriptingFloatChannel}
          Scale.X {type=MovieSceneScriptingFloatChannel, keys [[0,1],[3,1.125],[4,1]]}
          Scale.Y {type=MovieSceneScriptingFloatChannel, keys [[0,1],[3,1.125],[4,1]]}
          Shear.X {type=MovieSceneScriptingFloatChannel}
          Shear.Y {type=MovieSceneScriptingFloatChannel}
         ]
        }]
      }
      {
       type=MovieSceneColorTrack
       display_name=Color and Opacity
       sections [{
         type=MovieSceneColorSection
         end_frame=3
         start_frame=0
         channels [
          Color.R {type=MovieSceneScriptingFloatChannel, keys [[0,0.5089],[3,1]]}
          Color.G {type=MovieSceneScriptingFloatChannel, keys [[0,0.5583],[3,1]]}
          Color.B {type=MovieSceneScriptingFloatChannel, keys [[0,0.7084],[3,1]]}
          Color.A {type=MovieSceneScriptingFloatChannel, const [1,1]}
         ]
        }]
      }
     ]
    }
    AnimatedInputBorder {
     id=#2
     possessed_class=Border
     tracks [
      {
       type=MovieSceneWidgetMaterialTrack
       sections [{
         type=MovieSceneParameterSection
         channels [
          AnimateStateChange {type=MovieSceneScriptingFloatChannel, keys [[0,0],[3,1],[4,1]]}
          RGBR {type=MovieSceneScriptingFloatChannel, const [0.0021,1]}
          RGBG {type=MovieSceneScriptingFloatChannel, const [0.0497,1]}
          RGBB {type=MovieSceneScriptingFloatChannel, const [0.3712,1]}
          RGBA {type=MovieSceneScriptingFloatChannel, const [1,1]}
         ]
        }]
      }
      {
       type=MovieSceneColorTrack
       display_name=Brush Color
       sections [{
         type=MovieSceneColorSection
         channels [
          Color.R {type=MovieSceneScriptingFloatChannel}
          Color.G {type=MovieSceneScriptingFloatChannel}
          Color.B {type=MovieSceneScriptingFloatChannel}
          Color.A {type=MovieSceneScriptingFloatChannel, keys [[0,0],[2,1]]}
         ]
        }]
      }
     ]
    }
    Text_ListName {
     id=#3
     possessed_class=CommonTextBlock
     tracks [{
       type=MovieSceneColorTrack
       display_name=Color and Opacity
       sections [{
         type=MovieSceneColorSection
         end_frame=2
         start_frame=0
         channels [
          Color.R {type=MovieSceneScriptingFloatChannel, keys [[0,0.5089],[2,1]]}
          Color.G {type=MovieSceneScriptingFloatChannel, keys [[0,0.5583],[2,1]]}
          Color.B {type=MovieSceneScriptingFloatChannel, keys [[0,0.7084],[2,1]]}
          Color.A {type=MovieSceneScriptingFloatChannel, const [1,2]}
         ]
        }]
      }]
    }
    Text_NumEntries {
     id=#4
     possessed_class=CommonTextBlock
     tracks [{
       type=MovieSceneColorTrack
       display_name=Color and Opacity
       sections [{
         type=MovieSceneColorSection
         end_frame=4
         start_frame=0
         channels [
          Color.R {type=MovieSceneScriptingFloatChannel, keys [[0,0.5089],[4,1]]}
          Color.G {type=MovieSceneScriptingFloatChannel, keys [[0,0.5583],[4,1]]}
          Color.B {type=MovieSceneScriptingFloatChannel, keys [[0,0.7084],[4,1]]}
          Color.A {type=MovieSceneScriptingFloatChannel, const [1,1]}
         ]
        }]
      }]
    }
    EntryBorder {
     id=#5
     possessed_class=Border
     tracks [
      {
       type=MovieSceneWidgetMaterialTrack
       sections [{
         type=MovieSceneParameterSection
         end_frame=4
         start_frame=0
         channels [PrimaryAlpha {type=MovieSceneScriptingFloatChannel, keys [[0,0],[3,0.5],[4,0.5]]}]
        }]
      }
      {
       type=MovieScene2DTransformTrack
       sections [{
         type=MovieScene2DTransformSection
         end_frame=4
         start_frame=0
         channels [
          Angle {type=MovieSceneScriptingFloatChannel}
          Scale.X {type=MovieSceneScriptingFloatChannel}
          Scale.Y {type=MovieSceneScriptingFloatChannel}
          Shear.X {type=MovieSceneScriptingFloatChannel}
          Shear.Y {type=MovieSceneScriptingFloatChannel}
         ]
        }]
      }
      {
       type=MovieSceneColorTrack
       display_name=Brush Color
       sections [{
         type=MovieSceneColorSection
         end_frame=0
         start_frame=0
         channels [
          Color.R {type=MovieSceneScriptingFloatChannel, const [1,1]}
          Color.G {type=MovieSceneScriptingFloatChannel, const [1,1]}
          Color.B {type=MovieSceneScriptingFloatChannel, const [1,1]}
          Color.A {type=MovieSceneScriptingFloatChannel, const [1,1]}
         ]
        }]
      }
     ]
    }
    HorizontalBoxSlot_2 {id=#6, parent_id=#3, possessed_class=HorizontalBoxSlot, tracks []}
    InputActionWidget {
     id=#7
     possessed_class=CommonActionWidget
     tracks [{
       type=MovieSceneFloatTrack
       display_name=Render Opacity
       sections [{type=MovieSceneFloatSection, end_frame=4, start_frame=0, channels [None {type=MovieSceneScriptingFloatChannel, keys [[0,0],[1,0],[2,1]]}]}]
      }]
    }
   ]}
 }
 OnExpansion {
  binding_count=1
  display_label=""
  length_seconds=0.2
  object=&self:OnExpansion
  sequence {bindings [ExpandDirectionTick {
     id=#8
     possessed_class=Image
     tracks [{
       type=MovieScene2DTransformTrack
       sections [{
         type=MovieScene2DTransformSection
         channels [
          Angle {type=MovieSceneScriptingFloatChannel, keys [[0,0],[4,90]]}
          Scale.X {type=MovieSceneScriptingFloatChannel, keys [[0,1],[3,1.5],[4,1]]}
          Scale.Y {type=MovieSceneScriptingFloatChannel, keys [[0,1],[3,1.5],[4,1]]}
          Shear.X {type=MovieSceneScriptingFloatChannel}
          Shear.Y {type=MovieSceneScriptingFloatChannel}
         ]
        }]
      }]
    }]}
 }
]
graphs [EventGraph {nodes [
   Event On Hovered()
   > StopAllAnimations(self:User Widget Object Reference)
   > PlayAnimation(InAnimation=Get OnHover() -> (OnHover:Widget Animation Object Reference))
   > SetFocusedInputDisplay!(execute, self:Fort Sidebar Lfg Filter Header Entry Object Reference, bIsShown:bool=true) -> (then)
   n1 = Event On Unhovered()
   PlayAnimation(execute=[@n1.then, @n3.then], InAnimation=Get OnHover(), PlayMode=Reverse)
   > SetFocusedInputDisplay(bIsShown=false)
   n2 = Event On Item Expansion Changed!() -> (OutputDelegate, then, bIsExpanded:bool) {is=title K2Node_Event}
   > PlayAnimation(
    InAnimation=Get OnExpansion() -> (OnExpansion:Widget Animation Object Reference)
    PlayMode=Select(Option 0:EUMGSequencePlayMode Enum=Reverse, Option 1:EUMGSequencePlayMode Enum, Index=@n2.bIsExpanded) -> (ReturnValue:EUMGSequencePlayMode Enum)
    PlaybackSpeed=1.4
   )
   > PlaySound2D(Sound=/Game/Sounds/Fort_UI_Sounds/FrontEnd/New_FrontEnd/Fortnite_UI_MainMenu_Tab_Hover_02_Cue)
   n3 = Event OnFocusLost!() -> (OutputDelegate, then, InFocusEvent:Focus Event Structure) {is=title K2Node_Event}
   K2Node_Event_6 = OnListItemObjectSet_1!() -> (OutputDelegate, then, ListItemObject:Object Reference) {is=title K2Node_CustomEvent}
   > Branch(Condition=IsListItemExpanded!(UserListEntry:User List Entry Interface) -> (ReturnValue:bool))
   n4 = Get OnExpansion()
   > PlayAnimation(InAnimation=@n4, StartAtTime=GetEndTime(self=@n4))
   > SetFocus()
  ]}]
variables []
widgets {
 info {inheritedWidgetCount=5, parentClass=/Script/FortniteUI.FortSidebarLfgFilterHeaderEntry, rootWidgetClass=&SizeBox, widgetCount=13}
 widgets [
  SB_EntryContent {
   widgetClassPath=&SizeBox
   bIsVariable=true
   widget=SB_EntryContent
   properties {bOverride_MinDesiredHeight=true, heightOverride=40, minDesiredHeight=40}
  }
  EntryBorder {
   widgetClassPath=&Border
   parent=SB_EntryContent
   widget=EntryBorder
   properties {
    background {
     bIsDynamicallyLoaded=false
     drawAs=Image
     imageType=NoImage
     mirroring=NoMirror
     resourceName=None
     resourceObject=null
     tiling=NoTile
     imageSize=<32 32>
     margin=<ltrb 0 0 0 0>
     outlineSettings {
      bUseBrushTransparency=false
      roundingType=HalfHeightRadius
      width=0
      color {colorUseRule=UseColor_Specified, specifiedColor=<rgba 0 0 0 0>}
      cornerRadii=<0 0 0 0>
     }
     tintColor {colorUseRule=UseColor_Specified, specifiedColor=<rgba 0.0037 0.0052 0.016 1>}
     uVRegion {bIsValid=false, max=<0 0>, min=<0 0>}
    }
    padding=<ltrb 0 0 0 0>
   }
   slot {class=SizeBoxSlot}
  }
  HorizontalBox_0 {
   widgetClassPath=&HorizontalBox
   parent=EntryBorder
   widget=HorizontalBox_0
   slot {class=BorderSlot, properties {padding=<ltrb 0 0 0 0>}}
  }
  Text_ListName {
   widgetClassPath=&CommonTextBlock
   bInherited=true
   bIsVariable=true
   parent=HorizontalBox_0
   widget=Text_ListName
   properties {
    mobileFontSizeMultiplier=1.34
    scrollStyle=&TextScrollStyle-Base_C
    text="{CATEGORY NAME}"
    textTransformPolicy=ToUpper
    colorAndOpacity {colorUseRule=UseColor_Specified, specifiedColor=<rgba 0.5089 0.5583 0.7084 1>}
    font {
     bForceMonospaced=false
     bMaterialIsStencil=false
     fontMaterial=null
     fontObject=&BurbankSmall
     letterSpacing=0
     monospacedWidth=1
     size=15
     skewAmount=0
     typefaceFontName=Black
     outlineSettings {
      bApplyOutlineToDropShadows=false
      bMiteredCorners=false
      bSeparateFillAlpha=false
      outlineMaterial=null
      outlineSize=0
      outlineColor=<rgba 0 0 0 1>
     }
    }
    renderTransform {angle=0, scale=<1 1>, shear=<-8 0>, translation=<0 0>}
   }
   slot {class=HorizontalBoxSlot, properties {verticalAlignment=VAlign_Center, padding=<ltrb 24 0 0 0>}}
  }
  Text_NumEntries {
   widgetClassPath=&CommonTextBlock
   bInherited=true
   bIsVariable=true
   parent=HorizontalBox_0
   widget=Text_NumEntries
   properties {
    mobileFontSizeMultiplier=1.34
    text="{N}"
    colorAndOpacity {colorUseRule=UseColor_Specified, specifiedColor=<rgba 0.5089 0.5583 0.7084 1>}
    font {
     bForceMonospaced=false
     bMaterialIsStencil=false
     fontMaterial=null
     fontObject=&BurbankSmall
     letterSpacing=0
     monospacedWidth=1
     size=15
     skewAmount=0
     typefaceFontName=Black
     outlineSettings {
      bApplyOutlineToDropShadows=false
      bMiteredCorners=false
      bSeparateFillAlpha=false
      outlineMaterial=null
      outlineSize=0
      outlineColor=<rgba 0 0 0 1>
     }
    }
   }
   slot {class=HorizontalBoxSlot, properties {verticalAlignment=VAlign_Center, padding=<ltrb 3 0 0 0>, size {sizeRule=Fill, value=1}}}
  }
  HoverBubble {
   widgetClassPath=&SizeBox
   parent=HorizontalBox_0
   widget=HoverBubble
   properties {bOverride_MinDesiredWidth=true, minDesiredWidth=80}
   slot {class=HorizontalBoxSlot, properties {padding=<ltrb 20 0 0 0>}}
  }
  AnimatedInputBorder {
   widgetClassPath=&Border
   parent=HoverBubble
   widget=AnimatedInputBorder
   properties {
    horizontalAlignment=HAlign_Right
    visibility=SelfHitTestInvisible
    background {
     bIsDynamicallyLoaded=false
     drawAs=Image
     imageType=NoImage
     mirroring=NoMirror
     resourceName=None
     resourceObject=/Game/UI/SocialPanel/Art/Materials/M_UI_UserHoverCategory_Inst
     tiling=NoTile
     imageSize=<64 64>
     margin=<ltrb 1 0 0 0>
     outlineSettings {
      bUseBrushTransparency=false
      roundingType=HalfHeightRadius
      width=0
      color {colorUseRule=UseColor_Specified, specifiedColor=<rgba 0 0 0 0>}
      cornerRadii=<0 0 0 0>
     }
     tintColor {colorUseRule=UseColor_Specified, specifiedColor=<rgba 1 1 1 1>}
     uVRegion {bIsValid=false, max=<0 0>, min=<0 0>}
    }
    brushColor=<rgba 1 1 1 0>
    padding=<ltrb 24 0 24 0>
   }
   slot {class=SizeBoxSlot}
  }
  HorizontalBox_6 {
   widgetClassPath=&HorizontalBox
   parent=AnimatedInputBorder
   widget=HorizontalBox_6
   slot {class=BorderSlot, properties {horizontalAlignment=HAlign_Right, padding=<ltrb 24 0 24 0>}}
  }
  InputActionWidget {
   widgetClassPath=&CommonActionWidget
   bInherited=true
   parent=HorizontalBox_6
   widget=InputActionWidget
   properties {
    visibility=SelfHitTestInvisible
    inputActions [{dataTable=&InputActionDataTable, rowName=GenericFaceButtonLeft}]
    progressMaterialBrush {
     bIsDynamicallyLoaded=false
     drawAs=Image
     imageType=NoImage
     mirroring=NoMirror
     resourceName=None
     resourceObject=null
     tiling=NoTile
     imageSize=<16 16>
     margin=<ltrb 0 0 0 0>
     outlineSettings {
      bUseBrushTransparency=false
      roundingType=HalfHeightRadius
      width=0
      color {colorUseRule=UseColor_Specified, specifiedColor=<rgba 0 0 0 0>}
      cornerRadii=<0 0 0 0>
     }
     tintColor {colorUseRule=UseColor_Specified, specifiedColor=<rgba 1 1 1 1>}
     uVRegion {bIsValid=false, max=<0 0>, min=<0 0>}
    }
   }
   slot {class=HorizontalBoxSlot, properties {horizontalAlignment=HAlign_Left, verticalAlignment=VAlign_Center, padding=<ltrb 0 0 10 0>}}
  }
  FocusedInputActionWidget {
   widgetClassPath=&CommonActionWidget
   bInherited=true
   parent=HorizontalBox_6
   widget=FocusedInputActionWidget
   properties {
    visibility=Collapsed
    inputActions [{dataTable=&InputActionDataTable, rowName=Select}]
    progressMaterialBrush {
     bIsDynamicallyLoaded=false
     drawAs=Image
     imageType=NoImage
     mirroring=NoMirror
     resourceName=None
     resourceObject=null
     tiling=NoTile
     imageSize=<16 16>
     margin=<ltrb 0 0 0 0>
     outlineSettings {
      bUseBrushTransparency=false
      roundingType=HalfHeightRadius
      width=0
      color {colorUseRule=UseColor_Specified, specifiedColor=<rgba 0 0 0 0>}
      cornerRadii=<0 0 0 0>
     }
     tintColor {colorUseRule=UseColor_Specified, specifiedColor=<rgba 1 1 1 1>}
     uVRegion {bIsValid=false, max=<0 0>, min=<0 0>}
    }
   }
   slot {class=HorizontalBoxSlot, properties {horizontalAlignment=HAlign_Left, verticalAlignment=VAlign_Center, padding=<ltrb 0 0 10 0>}}
  }
  UnfocusedInputActionWidget {
   widgetClassPath=&CommonActionWidget
   bInherited=true
   parent=HorizontalBox_6
   widget=UnfocusedInputActionWidget
   properties {inputActions [{dataTable=&InputActionDataTable, rowName=GenericFaceButtonLeft}]}
   slot {class=HorizontalBoxSlot}
  }
  ArrowSpacingSB {
   widgetClassPath=&SizeBox
   parent=HorizontalBox_6
   widget=ArrowSpacingSB
   properties {bOverride_MinDesiredHeight=true, heightOverride=72, minDesiredHeight=16, widthOverride=72}
   slot {class=HorizontalBoxSlot, properties {horizontalAlignment=HAlign_Left, verticalAlignment=VAlign_Center, padding=<ltrb 10 0 0 0>}}
  }
  ExpandDirectionTick {
   widgetClassPath=/Script/FortniteUI.FortMobileImage
   bIsVariable=true
   parent=ArrowSpacingSB
   widget=ExpandDirectionTick
   properties {
    mobileSizeMultiplier=1.2
    brush {
     bIsDynamicallyLoaded=false
     drawAs=Image
     imageType=NoImage
     mirroring=NoMirror
     resourceName=None
     resourceObject=/Game/UI/SocialPanel/Art/Textures/T_UI_Icon_TreeArrow_64
     tiling=NoTile
     imageSize=<16 16>
     margin=<ltrb 0 0 0 0>
     outlineSettings {
      bUseBrushTransparency=false
      roundingType=HalfHeightRadius
      width=0
      color {colorUseRule=UseColor_Specified, specifiedColor=<rgba 0 0 0 0>}
      cornerRadii=<0 0 0 0>
     }
     tintColor {colorUseRule=UseColor_Specified, specifiedColor=<rgba 1 1 1 1>}
     uVRegion {bIsValid=false, max=<0 0>, min=<0 0>}
    }
    colorAndOpacity=<rgba 0.5089 0.5583 0.7084 1>
    renderTransform {angle=90, scale=<1 1>, shear=<0 0>, translation=<0 0>}
   }
   slot {class=SizeBoxSlot, properties {horizontalAlignment=HAlign_Left, verticalAlignment=VAlign_Center}}
  }
 ]
}
