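# Sidebar list entry for a social tag filter, built on FortLFGFilterCheckboxEntryWidget with a tag widget, Select input hint and checkbox.
# Unless touch is in use, hover and unhover animate background highlight, padding, checkbox color and hint opacity, then notify the tag widget.
# Construct enlarges the checkbox and raises minimum row height on mobile.
# Input icon updates set the hint's dynamic material to dark without an outline.
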WidgetBlueprint /FortniteCoreUI/UI/Sidebar/Tags/SidebarSocialTagFilterCheckboxListEntry
animations [
 Hovered {
  binding_count=6
  length_seconds=0.2
  object=&self:Hovered
  sequence {bindings [
    Border_Background {
     id=#1
     possessed_class=Border
     tracks [{
       type=MovieSceneWidgetMaterialTrack
       sections [{
         type=MovieSceneParameterSection
         end_frame=4
         start_frame=0
         channels [AnimateHighlight {type=MovieSceneScriptingFloatChannel, keys [[0,0],[3,1],[4,1]]}]
        }]
      }]
    }
    BorderSlot_2 {
     id=#2
     parent_id=#3
     possessed_class=BorderSlot
     tracks [{
       type=MovieSceneMarginTrack
       display_name=Padding
       sections [{
         type=MovieSceneMarginSection
         end_frame=2
         start_frame=0
         channels [
          Left {type=MovieSceneScriptingFloatChannel, const [0,2]}
          Top {type=MovieSceneScriptingFloatChannel, const [0,2]}
          Right {type=MovieSceneScriptingFloatChannel, keys [[0,0],[2,10]]}
          Bottom {type=MovieSceneScriptingFloatChannel, const [0,2]}
         ]
        }]
      }]
    }
    HorizontalBoxSlot_1 {
     id=#4
     parent_id=#5
     possessed_class=HorizontalBoxSlot
     tracks [{
       type=MovieSceneMarginTrack
       display_name=Padding
       sections [{
         type=MovieSceneMarginSection
         end_frame=3
         start_frame=0
         channels [
          Left {type=MovieSceneScriptingFloatChannel, keys [[0,15],[3,20]]}
          Top {type=MovieSceneScriptingFloatChannel, const [0,1]}
          Right {type=MovieSceneScriptingFloatChannel, const [0,1]}
          Bottom {type=MovieSceneScriptingFloatChannel, const [0,1]}
         ]
        }]
      }]
    }
    SocialTagWidget {id=#5, possessed_class=SidebarSocialTagWidget_C, tracks []}
    InputActionWidget {
     id=#3
     possessed_class=CommonActionWidget
     tracks [{
       type=MovieSceneFloatTrack
       display_name=Render Opacity
       sections [{type=MovieSceneFloatSection, end_frame=4, start_frame=0, channels [None {type=MovieSceneScriptingFloatChannel, keys [[0,0],[2,1]]}]}]
      }]
    }
    Checkbox {
     id=#6
     possessed_class=FortCheckbox_C
     tracks [{
       type=MovieSceneColorTrack
       display_name=Color and Opacity
       sections [{
         type=MovieSceneColorSection
         end_frame=2
         start_frame=0
         channels [
          Color.R {type=MovieSceneScriptingFloatChannel, keys [[0,1],[2,0]]}
          Color.G {type=MovieSceneScriptingFloatChannel, keys [[0,1],[2,0]]}
          Color.B {type=MovieSceneScriptingFloatChannel, keys [[0,1],[2,0]]}
          Color.A {type=MovieSceneScriptingFloatChannel, const [1,2]}
         ]
        }]
      }]
    }
   ]}
 }
 Unhovered {
  binding_count=6
  length_seconds=0.2
  object=&self:Unhovered
  sequence {bindings [
    Checkbox {
     id=#7
     possessed_class=FortCheckbox_C
     tracks [{
       type=MovieSceneColorTrack
       display_name=Color and Opacity
       sections [{
         type=MovieSceneColorSection
         end_frame=4
         start_frame=0
         channels [
          Color.R {type=MovieSceneScriptingFloatChannel, keys [[-0.0113,0],[2,0],[4,1]]}
          Color.G {type=MovieSceneScriptingFloatChannel, keys [[0,0],[2,0],[4,1]]}
          Color.B {type=MovieSceneScriptingFloatChannel, keys [[0,0],[2,0],[4,1]]}
          Color.A {type=MovieSceneScriptingFloatChannel, const [1,3]}
         ]
        }]
      }]
    }
    Border_Background {
     id=#1
     possessed_class=Border
     tracks [{
       type=MovieSceneWidgetMaterialTrack
       sections [{
         type=MovieSceneParameterSection
         end_frame=4
         start_frame=0
         channels [AnimateHighlight {type=MovieSceneScriptingFloatChannel, keys [[0,1],[1,1],[4,0]]}]
        }]
      }]
    }
    HorizontalBoxSlot_1 {
     id=#8
     parent_id=#9
     possessed_class=HorizontalBoxSlot
     tracks [{
       type=MovieSceneMarginTrack
       display_name=Padding
       sections [{
         type=MovieSceneMarginSection
         end_frame=4
         start_frame=0
         channels [
          Left {type=MovieSceneScriptingFloatChannel, keys [[-0.0113,20],[1,20],[4,15]]}
          Top {type=MovieSceneScriptingFloatChannel}
          Right {type=MovieSceneScriptingFloatChannel}
          Bottom {type=MovieSceneScriptingFloatChannel}
         ]
        }]
      }]
    }
    SocialTagWidget {id=#9, possessed_class=SidebarSocialTagWidget_C, tracks []}
    BorderSlot_2 {
     id=#2
     parent_id=#3
     possessed_class=BorderSlot
     tracks [{
       type=MovieSceneMarginTrack
       display_name=Padding
       sections [{
         type=MovieSceneMarginSection
         end_frame=2
         start_frame=0
         channels [
          Left {type=MovieSceneScriptingFloatChannel, const [0,1]}
          Top {type=MovieSceneScriptingFloatChannel, const [0,1]}
          Right {type=MovieSceneScriptingFloatChannel, keys [[0,10],[2,0]]}
          Bottom {type=MovieSceneScriptingFloatChannel, const [0,1]}
         ]
        }]
      }]
    }
    InputActionWidget {
     id=#3
     possessed_class=CommonActionWidget
     tracks [{
       type=MovieSceneFloatTrack
       display_name=Render Opacity
       sections [{type=MovieSceneFloatSection, end_frame=4, start_frame=0, channels [None {type=MovieSceneScriptingFloatChannel, keys [[1.9517,1],[4,0]]}]}]
      }]
    }
   ]}
 }
]
graphs [EventGraph {nodes [
   "On Input Icon Updated (InputActionWidget)"() {is=title K2Node_ComponentBoundEvent}
   > n1 = GetIconDynamicMaterial!(execute, self:Common Action Widget Object Reference=Get InputActionWidget() -> (InputActionWidget:Common Action Widget Object Reference)) -> (then, ReturnValue:Material Instance Dynamic Object Reference)
   > SetScalarParameterValue(self=@n1, ParameterName=Use Dark, Value=1)
   > SetScalarParameterValue(self=@n1, ParameterName=Outline Opacity, Value=0)
   Event Construct()
   > Branch(Condition=IsMobileGame())
   > SetRenderScale!(
    execute
    self:Widget Object Reference=Get SBox_Checkbox() -> (SBox_Checkbox:Size Box Object Reference)
    Scale:Vector 2D Structure=(X=1.5,Y=1.5)
   ) -> (then)
   > SetMinDesiredHeight!(execute, self:Size Box Object Reference=Get SizeBox_0() -> (SizeBox_0:Size Box Object Reference), InMinDesiredHeight:float=82) -> (then)
   Event On Hovered()
   > n2 = Branch(Condition=IsUsingTouch(PlayerController=GetOwningPlayer()))
   StopAllAnimations(execute=@n2.else, self:User Widget Object Reference)
   > PlayAnimation(InAnimation=Get Hovered() -> (Hovered:Widget Animation Object Reference))
   n3 = Get SocialTagWidget() -> (SocialTagWidget:Social Tag Widget Object Reference)
   > n4 = Cast To SidebarSocialTagWidget!(execute, Object=@n3) -> (then, CastFailed, AsSidebar Social Tag Widget:Sidebar Social Tag Widget Object Reference) {is=title K2Node_DynamicCast}
   > OnTagListEntryHovered(self:Sidebar Social Tag Widget Object Reference=@n4)
   Event On Unhovered()
   > n5 = Branch(Condition=IsUsingTouch(PlayerController=GetOwningPlayer()))
   StopAllAnimations(execute=@n5.else)
   > PlayAnimation(InAnimation=Get Unhovered() -> (Unhovered:Widget Animation Object Reference), PlaybackSpeed=2)
   > n6 = Cast To SidebarSocialTagWidget(Object=@n3)
   > OnTagListEntryUnhovered(self:Sidebar Social Tag Widget Object Reference=@n6)
  ]}]
variables [
 EncourageEpicFriend_HoveredColor {type=LinearColor, default=<rgba 1 1 1 1>}
 EncourageEpicFriend_IdleColor {type=LinearColor, default=<rgba 0.9558 1 0.0672 1>}
]
widgets {
 info {inheritedWidgetCount=4, parentClass=/Script/FortniteUI.FortLFGFilterCheckboxEntryWidget, rootWidgetClass=&SizeBox, widgetCount=8}
 widgets [
  SizeBox_0 {
   widgetClassPath=&SizeBox
   bIsVariable=true
   widget=SizeBox_0
   properties {bOverride_MinDesiredHeight=true, heightOverride=72, minDesiredHeight=72}
  }
  Border_Background {
   widgetClassPath=&Border
   bIsVariable=true
   parent=SizeBox_0
   widget=Border_Background
   properties {
    clipping=ClipToBounds
    visibility=SelfHitTestInvisible
    background {
     bIsDynamicallyLoaded=false
     drawAs=Image
     imageType=NoImage
     mirroring=NoMirror
     resourceName=None
     resourceObject=/Game/UI/Sidebar/Materials/M_UI_sidebar_Button_Selection_Inst
     tiling=NoTile
     imageSize=<32 32>
     margin=<ltrb 0 0 0 0>
     outlineSettings {
      bUseBrushTransparency=false
      roundingType=HalfHeightRadius
      width=0
      color {colorUseRule=UseColor_Specified, specifiedColor=<rgba 0 0 0 0>}
      cornerRadii=<0 0 0 0>
     }
     tintColor {colorUseRule=UseColor_Specified, specifiedColor=<rgba 1 1 1 1>}
     uVRegion {bIsValid=false, max=<0 0>, min=<0 0>}
    }
    padding=<ltrb 0 0 0 0>
   }
   slot {class=SizeBoxSlot}
  }
  HorizontalBox_2 {
   widgetClassPath=&HorizontalBox
   parent=Border_Background
   widget=HorizontalBox_2
   properties {visibility=HitTestInvisible}
   slot {class=BorderSlot, properties {padding=<ltrb 0 0 0 0>}}
  }
  SocialTagWidget {
   widgetClassPath=/Game/UI/Sidebar/Tags/SidebarSocialTagWidget
   bInherited=true
   bIsVariable=true
   parent=HorizontalBox_2
   widget=SocialTagWidget
   properties {bApplyAlphaOnDisable=false, bIsFocusable=false, bIsSelectable=false, bSelectable=false}
   slot {class=HorizontalBoxSlot, properties {horizontalAlignment=HAlign_Left, padding=<ltrb 15 0 0 0>, size {sizeRule=Fill, value=1}}}
  }
  InputActionWidget {
   widgetClassPath=&CommonActionWidget
   bInherited=true
   parent=HorizontalBox_2
   widget=InputActionWidget
   properties {renderOpacity=0, visibility=SelfHitTestInvisible, inputActions [{dataTable=&InputActionDataTable, rowName=Select}]}
   slot {class=HorizontalBoxSlot, properties {horizontalAlignment=HAlign_Right, verticalAlignment=VAlign_Center}}
  }
  SBox_Checkbox {
   widgetClassPath=&SizeBox
   bIsVariable=true
   parent=HorizontalBox_2
   widget=SBox_Checkbox
   properties {bOverride_HeightOverride=true, bOverride_WidthOverride=true, heightOverride=32, widthOverride=32, renderTransformPivot=<1 0.5>}
   slot {class=HorizontalBoxSlot, properties {horizontalAlignment=HAlign_Right, verticalAlignment=VAlign_Center, padding=<ltrb 0 0 20 0>}}
  }
  Checkbox {
   widgetClassPath=&/FortSidebarCheckbox
   bInherited=true
   bIsVariable=true
   parent=SBox_Checkbox
   widget=Checkbox
   slot {class=SizeBoxSlot, properties {padding=<ltrb -16 -3.5 -16 -3.5>}}
  }
  MenuAnchor_Actions {
   widgetClassPath=/Script/UMG.MenuAnchor
   bInherited=true
   bIsVariable=true
   parent=HorizontalBox_2
   widget=MenuAnchor_Actions
   properties {menuClass=/Game/UI/SocialPanel/UserActionMenu.UserActionMenu_C, placement=MenuPlacement_MenuRight}
   slot {class=HorizontalBoxSlot}
  }
 ]
}
