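# Widget Blueprint for the Reboot Rally sidebar that shows the time remaining until the Reboot Rally ends, built on the native RebootRallyTimerBase class.
# A grid lays a timer icon beside a CommonTextBlock over a dark UIKit backplate, inside an overlay that also holds a fill-aligned post-buffer capture widget.
# The UpdateRemainingTime event writes the parent's GetRemainingTimeForRebootRallyEnd text into Text_Timer; PreConstruct, Construct and Tick are empty.
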WidgetBlueprint /Game/UI/Sidebar/RebootRally/WBP_RebootRallyTimer
graphs [EventGraph {nodes [
   Event UpdateRemainingTime() {is=title K2Node_Event}
   > Set Text(
    Text=Get RemainingTimeForRebootRallyEnd() -> (RemainingTimeForRebootRallyEnd:Text)
    self=Get Text_Timer() -> (Text_Timer:Common Text Object Reference)
   )
   Event PreConstruct()
   Event Construct()
   Event Tick()
  ]}]
variables []
widgets {
 info {inheritedWidgetCount=0, parentClass=/Script/FortniteUI.RebootRallyTimerBase, rootWidgetClass=&Overlay, widgetCount=6}
 widgets [
  Overlay_2 {widgetClassPath=&Overlay, widget=Overlay_2}
  WBP_CaptureForPostBufferUpdate {
   widgetClassPath=&WBP_CaptureForPostBufferUpdate
   bIsVariable=true
   parent=Overlay_2
   slot {class=OverlaySlot, properties {horizontalAlignment=HAlign_Fill, verticalAlignment=VAlign_Fill}}
  }
  GridPanel_125 {
   widgetClassPath=&GridPanel
   parent=Overlay_2
   widget=GridPanel_125
   properties {columnFill [0,0], rowFill [0]}
   slot {class=OverlaySlot}
  }
  WBP_UIKit_Backplate {
   widgetClassPath=&WBP_UIKit_Backplate
   bIsVariable=true
   parent=GridPanel_125
   widget=WBP_UIKit_Backplate
   properties {brightness=Dark}
   slot {class=GridSlot, properties {columnSpan=2, layer=-1}}
  }
  Image_Timer {
   widgetClassPath=&Image
   bIsVariable=true
   parent=GridPanel_125
   widget=Image_Timer
   properties {brush {
     bIsDynamicallyLoaded=false
     drawAs=Image
     imageType=NoImage
     mirroring=NoMirror
     resourceName=None
     resourceObject=/Game/Athena/UI/Frontend/ActivityBrowser/Art/M_UI_Timer_Icon_ShopOffer
     tiling=NoTile
     imageSize=<24 24>
     margin=<ltrb 0 0 0 0>
     outlineSettings {
      bUseBrushTransparency=false
      roundingType=HalfHeightRadius
      width=0
      color {colorUseRule=UseColor_Specified, specifiedColor=<rgba 0 0 0 0>}
      cornerRadii=<0 0 0 0>
     }
     tintColor {colorUseRule=UseColor_Specified, specifiedColor=<rgba 1 1 1 1>}
     uVRegion {bIsValid=false, max=<0 0>, min=<0 0>}
    }}
   slot {class=GridSlot, properties {horizontalAlignment=HAlign_Right, layer=1, verticalAlignment=VAlign_Center, padding=<ltrb 6 0 0 0>}}
  }
  Text_Timer {
   widgetClassPath=&CommonTextBlock
   bIsVariable=true
   parent=GridPanel_125
   widget=Text_Timer
   properties {
    mobileFontSizeMultiplier=1.4
    text=0W 0D
    font {
     bForceMonospaced=false
     bMaterialIsStencil=false
     fontMaterial=/FortUILibrary/Materials/Text/M_UI_Additive_Font
     fontObject=&HeadingNow
     letterSpacing=0
     monospacedWidth=1
     size=13.5
     skewAmount=0
     typefaceFontName=75 Medium
     outlineSettings {
      bApplyOutlineToDropShadows=false
      bMiteredCorners=false
      bSeparateFillAlpha=false
      outlineMaterial=null
      outlineSize=0
      outlineColor=<rgba 0 0 0 1>
     }
    }
    shadowOffset=<0 0>
    strikeBrush {
     bIsDynamicallyLoaded=false
     drawAs=Image
     imageType=FullColor
     mirroring=NoMirror
     resourceName=None
     resourceObject=null
     tiling=NoTile
     imageSize=<32 32>
     margin=<ltrb 0 0 0 0>
     outlineSettings {
      bUseBrushTransparency=false
      roundingType=HalfHeightRadius
      width=0
      color {colorUseRule=UseColor_Specified, specifiedColor=<rgba 0 0 0 0>}
      cornerRadii=<0 0 0 0>
     }
     tintColor {colorUseRule=UseColor_Specified, specifiedColor=<rgba 1 1 1 1>}
     uVRegion {bIsValid=false, max=<0 0>, min=<0 0>}
    }
   }
   slot {class=GridSlot, properties {column=1, layer=1, verticalAlignment=VAlign_Center, padding=<ltrb 4 0 8 0>}}
  }
 ]
}
